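# Physics asset for the DistantEchoCastle Style female medium player body cosmetic, covering the dynamic skirt strap bones.
# Small capsule bodies form a long strap chain from a master bone through a base to three segments, with a short strap branching off the base.
# Constraints lock twist and allow symmetric swing that is tightest at the root and looser toward the tips and on the short strap.
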PhysicsAsset /BRCosmetics/Characters/Player/Female/Medium/Bodies/F_MED_DistantEchoCastle_Style/Meshes/F_MED_DistantEchoCastle_Style_Physics
bodies {
 dyn_skirt_strapLong_01 [{shapeType=Capsule, length=5, radius=1, shapeName=dyn_skirt_strapLong_01_capsule, center=<5 0 0>, rotation=<rot 90 0 0>}]
 dyn_skirt_strapLong_02 [{shapeType=Capsule, length=5, radius=1, shapeName=dyn_skirt_strapLong_02_capsule, center=<5 0 0>, rotation=<rot 90 0 0>}]
 dyn_skirt_strapLong_03 [{shapeType=Capsule, length=5, radius=1, shapeName=dyn_skirt_strapLong_03_capsule, center=<5 0 0>, rotation=<rot 90 0 0>}]
 dyn_skirt_strapLong_MSTR [{shapeType=Capsule, length=0.505, radius=0.505, shapeName=dyn_skirt_strapLong_MSTR_capsule, rotation=<rot 0 -90 0>}]
 dyn_skirt_strapLong_base [{shapeType=Capsule, length=5, radius=1, shapeName=dyn_skirt_strapLong_base_capsule, center=<5 0 0>, rotation=<rot 90 0 0>}]
 dyn_skirt_strapShort_01 [{shapeType=Capsule, length=5, radius=1, shapeName=dyn_skirt_strapShort_01_capsule, center=<5 0 0>, rotation=<rot 90 0 0>}]
}
constraints [
 {bone1Name=dyn_skirt_strapLong_base, bone2Name=dyn_skirt_strapLong_MSTR, swing1LimitDegrees=5, swing2LimitDegrees=5, twistLimitDegrees=0}
 {bone1Name=dyn_skirt_strapLong_01, bone2Name=dyn_skirt_strapLong_base, swing1LimitDegrees=10, swing2LimitDegrees=10, twistLimitDegrees=0}
 {bone1Name=dyn_skirt_strapLong_02, bone2Name=dyn_skirt_strapLong_01, swing1LimitDegrees=20, swing2LimitDegrees=20, twistLimitDegrees=0}
 {bone1Name=dyn_skirt_strapLong_03, bone2Name=dyn_skirt_strapLong_02, swing1LimitDegrees=20, swing2LimitDegrees=20, twistLimitDegrees=0}
 {bone1Name=dyn_skirt_strapShort_01, bone2Name=dyn_skirt_strapLong_base, swing1LimitDegrees=20, swing2LimitDegrees=20, twistLimitDegrees=0}
]
