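# Physics asset for the skirt of the female medium DistantEchoCastle_Style body cosmetic, with capsule bodies on the skirt bones, pelvis and both thighs.
# Nine skirt chains each run from a tiny MSTR capsule through base, 01 and 02 segments, with narrow swing limits and zero twist.
# Front and side chains anchor to their thighs and back chains to the pelvis, all with free angular motion.
# The left side root joint alone uses different swing limits.
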PhysicsAsset /BRCosmetics/Characters/Player/Female/Medium/Bodies/F_MED_DistantEchoCastle_Style/Meshes/F_MED_DistantEchoCastle_Style_Skirt_Physics
bodies {
 dyn_skirt_bk_01 [{shapeType=Capsule, length=8.25, radius=5, shapeName=dyn_skirt_bk_l_01_capsule, center=<2 0 5>, rotation=<rot 0 0 90>}]
 dyn_skirt_bk_02 [{shapeType=Capsule, length=8.25, radius=5, shapeName=dyn_skirt_bk_l_02_capsule, center=<2 0 5>, rotation=<rot 0 0 90>}]
 dyn_skirt_bk_MSTR [{shapeType=Capsule, length=0.505, radius=0.505, shapeName=dyn_skirt_bk_MSTR_capsule, rotation=<rot 0 -90 0>}]
 dyn_skirt_bk_base [{shapeType=Capsule, length=8.25, radius=5, shapeName=dyn_skirt_bk_l_base_capsule, center=<2 0 5>, rotation=<rot 0 0 90>}]
 dyn_skirt_bk_l_01 [{shapeType=Capsule, length=8.25, radius=5, shapeName=dyn_skirt_bk_l_01_capsule, center=<2 0 5>, rotation=<rot 0 0 90>}]
 dyn_skirt_bk_l_02 [{shapeType=Capsule, length=8.25, radius=5, shapeName=dyn_skirt_bk_l_02_capsule, center=<2 0 5>, rotation=<rot 0 0 90>}]
 dyn_skirt_bk_l_MSTR [{shapeType=Capsule, length=0.505, radius=0.505, shapeName=dyn_skirt_bk_l_MSTR_capsule, rotation=<rot 0 -90 0>}]
 dyn_skirt_bk_l_base [{shapeType=Capsule, length=8.25, radius=5, shapeName=dyn_skirt_bk_l_base_capsule, center=<2 0 5>, rotation=<rot 0 0 90>}]
 dyn_skirt_bk_r_01 [{shapeType=Capsule, length=8.25, radius=5, shapeName=dyn_skirt_bk_r_01_capsule, center=<-2 0 -5>, rotation=<rot 0 0 90>}]
 dyn_skirt_bk_r_02 [{shapeType=Capsule, length=8.25, radius=5, shapeName=dyn_skirt_bk_r_02_capsule, center=<-2 0 -5>, rotation=<rot 0 0 90>}]
 dyn_skirt_bk_r_MSTR [{shapeType=Capsule, length=0.505, radius=0.505, shapeName=dyn_skirt_bk_r_MSTR_capsule, rotation=<rot 0 -90 0>}]
 dyn_skirt_bk_r_base [{shapeType=Capsule, length=8.25, radius=5, shapeName=dyn_skirt_bk_r_base_capsule, center=<-2 0 -5>, rotation=<rot 0 0 90>}]
 dyn_skirt_l_front2_01 [{shapeType=Capsule, length=8.25, radius=5, shapeName=dyn_skirt_l_front2_01_capsule, center=<2 0 5>, rotation=<rot 0 0 90>}]
 dyn_skirt_l_front2_02 [{shapeType=Capsule, length=8.25, radius=5, shapeName=dyn_skirt_l_front2_02_capsule, center=<2 0 5>, rotation=<rot 0 0 90>}]
 dyn_skirt_l_front2_MSTR [{shapeType=Capsule, length=0.505, radius=0.505, shapeName=dyn_skirt_l_front2_MSTR_capsule, rotation=<rot 0 -90 0>}]
 dyn_skirt_l_front2_base [{shapeType=Capsule, length=8.25, radius=5, shapeName=dyn_skirt_l_front2_base_capsule, center=<2 0 5>, rotation=<rot 0 0 90>}]
 dyn_skirt_l_front_01 [{shapeType=Capsule, length=7.25, radius=5, shapeName=dyn_skirt_l_front_01_capsule, center=<2 0 5>, rotation=<rot 0 0 90>}]
 dyn_skirt_l_front_02 [{shapeType=Capsule, length=7.25, radius=6, shapeName=dyn_skirt_l_front_02_capsule, center=<2 0 5>, rotation=<rot 0 0 90>}]
 dyn_skirt_l_front_MSTR [{shapeType=Capsule, length=0.505, radius=0.505, shapeName=dyn_skirt_l_front_MSTR_capsule, rotation=<rot 0 -90 0>}]
 dyn_skirt_l_front_base [{shapeType=Capsule, length=8.75, radius=4.5, shapeName=dyn_skirt_l_front_base_capsule, center=<2 0 5>, rotation=<rot 0 0 90>}]
 dyn_skirt_l_side_01 [{shapeType=Capsule, length=8.75, radius=4.5, shapeName=dyn_skirt_l_side_01_capsule, center=<2 0 5>, rotation=<rot 0 0 90>}]
 dyn_skirt_l_side_02 [{shapeType=Capsule, length=8.25, radius=5, shapeName=dyn_skirt_l_side_02_capsule, center=<2 0 5>, rotation=<rot 0 0 90>}]
 dyn_skirt_l_side_MSTR [{shapeType=Capsule, length=0.505, radius=0.505, shapeName=dyn_skirt_l_side_MSTR_capsule, rotation=<rot 0 -90 0>}]
 dyn_skirt_l_side_base [{shapeType=Capsule, length=9.75, radius=3.5, shapeName=dyn_skirt_l_side_base_capsule, center=<2 0 5>, rotation=<rot 0 0 90>}]
 dyn_skirt_r_front2_01 [{shapeType=Capsule, length=8.25, radius=5, shapeName=dyn_skirt_r_front2_01_capsule, center=<-2 0 -5>, rotation=<rot 0 0 90>}]
 dyn_skirt_r_front2_02 [{shapeType=Capsule, length=8.25, radius=5, shapeName=dyn_skirt_r_front2_02_capsule, center=<-2 0 -5>, rotation=<rot 0 0 90>}]
 dyn_skirt_r_front2_MSTR [{shapeType=Capsule, length=0.505, radius=0.505, shapeName=dyn_skirt_r_front2_MSTR_capsule, rotation=<rot 0 -90 0>}]
 dyn_skirt_r_front2_base [{shapeType=Capsule, length=8.25, radius=5, shapeName=dyn_skirt_r_front2_base_capsule, center=<-2 0 -5>, rotation=<rot 0 0 90>}]
 dyn_skirt_r_front_01 [{shapeType=Capsule, length=7.25, radius=5, shapeName=dyn_skirt_r_front_01_capsule, center=<-2 0 -5>, rotation=<rot 0 0 90>}]
 dyn_skirt_r_front_02 [{shapeType=Capsule, length=7.25, radius=6, shapeName=dyn_skirt_r_front_02_capsule, center=<-2 0 -5>, rotation=<rot 0 0 90>}]
 dyn_skirt_r_front_MSTR [{shapeType=Capsule, length=0.505, radius=0.505, shapeName=dyn_skirt_r_front_MSTR_capsule, rotation=<rot 0 -90 0>}]
 dyn_skirt_r_front_base [{shapeType=Capsule, length=8.75, radius=4.5, shapeName=dyn_skirt_r_front_base_capsule, center=<-2 0 -5>, rotation=<rot 0 0 90>}]
 dyn_skirt_r_side_01 [{shapeType=Capsule, length=8.75, radius=4.5, shapeName=dyn_skirt_r_side_01_capsule, center=<-2 0 -5>, rotation=<rot 0 0 90>}]
 dyn_skirt_r_side_02 [{shapeType=Capsule, length=8.25, radius=5, shapeName=dyn_skirt_r_side_02_capsule, center=<-2 0 -5>, rotation=<rot 0 0 90>}]
 dyn_skirt_r_side_MSTR [{shapeType=Capsule, length=0.505, radius=0.505, shapeName=dyn_skirt_r_side_MSTR_capsule, rotation=<rot 0 -90 0>}]
 dyn_skirt_r_side_base [{shapeType=Capsule, length=9.75, radius=3.5, shapeName=dyn_skirt_r_side_base_capsule, center=<-2 0 -5>, rotation=<rot 0 0 90>}]
 pelvis [{shapeType=Capsule, length=7.75, radius=10, shapeName=pelvis_capsule, center=<0 -1.5521 0>}]
 thigh_l [{shapeType=Capsule, length=10.25, radius=9.75, shapeName=thigh_l_capsule, center=<-8 -1 2>, rotation=<rot 90 0 0>}]
 thigh_r [{shapeType=Capsule, length=10.25, radius=9.75, shapeName=thigh_r_capsule, center=<8 1 -2>, rotation=<rot 90 0 0>}]
}
constraints [
 {bone1Name=dyn_skirt_bk_l_base, bone2Name=dyn_skirt_bk_l_MSTR, swing1LimitDegrees=10, swing2LimitDegrees=20, twistLimitDegrees=0}
 {bone1Name=dyn_skirt_bk_l_01, bone2Name=dyn_skirt_bk_l_base, swing1LimitDegrees=10, swing2LimitDegrees=20, twistLimitDegrees=0}
 {bone1Name=dyn_skirt_bk_l_02, bone2Name=dyn_skirt_bk_l_01, swing1LimitDegrees=10, swing2LimitDegrees=20, twistLimitDegrees=0}
 {bone1Name=dyn_skirt_bk_r_base, bone2Name=dyn_skirt_bk_r_MSTR, swing1LimitDegrees=10, swing2LimitDegrees=20, twistLimitDegrees=0}
 {bone1Name=dyn_skirt_bk_r_01, bone2Name=dyn_skirt_bk_r_base, swing1LimitDegrees=10, swing2LimitDegrees=20, twistLimitDegrees=0}
 {bone1Name=dyn_skirt_bk_r_02, bone2Name=dyn_skirt_bk_r_01, swing1LimitDegrees=10, swing2LimitDegrees=20, twistLimitDegrees=0}
 {bone1Name=dyn_skirt_l_front2_base, bone2Name=dyn_skirt_l_front2_MSTR, swing1LimitDegrees=10, swing2LimitDegrees=20, twistLimitDegrees=0}
 {bone1Name=dyn_skirt_l_front2_01, bone2Name=dyn_skirt_l_front2_base, swing1LimitDegrees=10, swing2LimitDegrees=20, twistLimitDegrees=0}
 {bone1Name=dyn_skirt_l_front2_02, bone2Name=dyn_skirt_l_front2_01, swing1LimitDegrees=10, swing2LimitDegrees=20, twistLimitDegrees=0}
 {bone1Name=dyn_skirt_l_front_base, bone2Name=dyn_skirt_l_front_MSTR, swing1LimitDegrees=10, swing2LimitDegrees=20, twistLimitDegrees=0}
 {bone1Name=dyn_skirt_l_front_01, bone2Name=dyn_skirt_l_front_base, swing1LimitDegrees=10, swing2LimitDegrees=20, twistLimitDegrees=0}
 {bone1Name=dyn_skirt_l_front_02, bone2Name=dyn_skirt_l_front_01, swing1LimitDegrees=10, swing2LimitDegrees=20, twistLimitDegrees=0}
 {bone1Name=dyn_skirt_l_side_base, bone2Name=dyn_skirt_l_side_MSTR, swing1LimitDegrees=0, swing2LimitDegrees=40, twistLimitDegrees=0}
 {bone1Name=dyn_skirt_l_side_01, bone2Name=dyn_skirt_l_side_base, swing1LimitDegrees=10, swing2LimitDegrees=20, twistLimitDegrees=0}
 {bone1Name=dyn_skirt_l_side_02, bone2Name=dyn_skirt_l_side_01, swing1LimitDegrees=10, swing2LimitDegrees=20, twistLimitDegrees=0}
 {bone1Name=dyn_skirt_r_front2_base, bone2Name=dyn_skirt_r_front2_MSTR, swing1LimitDegrees=10, swing2LimitDegrees=20, twistLimitDegrees=0}
 {bone1Name=dyn_skirt_r_front2_01, bone2Name=dyn_skirt_r_front2_base, swing1LimitDegrees=10, swing2LimitDegrees=20, twistLimitDegrees=0}
 {bone1Name=dyn_skirt_r_front2_02, bone2Name=dyn_skirt_r_front2_01, swing1LimitDegrees=10, swing2LimitDegrees=20, twistLimitDegrees=0}
 {bone1Name=dyn_skirt_r_front_base, bone2Name=dyn_skirt_r_front_MSTR, swing1LimitDegrees=10, swing2LimitDegrees=20, twistLimitDegrees=0}
 {bone1Name=dyn_skirt_r_front_01, bone2Name=dyn_skirt_r_front_base, swing1LimitDegrees=10, swing2LimitDegrees=20, twistLimitDegrees=0}
 {bone1Name=dyn_skirt_r_front_02, bone2Name=dyn_skirt_r_front_01, swing1LimitDegrees=10, swing2LimitDegrees=20, twistLimitDegrees=0}
 {bone1Name=dyn_skirt_r_side_base, bone2Name=dyn_skirt_r_side_MSTR, swing1LimitDegrees=10, swing2LimitDegrees=20, twistLimitDegrees=0}
 {bone1Name=dyn_skirt_r_side_01, bone2Name=dyn_skirt_r_side_base, swing1LimitDegrees=10, swing2LimitDegrees=20, twistLimitDegrees=0}
 {bone1Name=dyn_skirt_r_side_02, bone2Name=dyn_skirt_r_side_01, swing1LimitDegrees=10, swing2LimitDegrees=20, twistLimitDegrees=0}
 {bone1Name=dyn_skirt_bk_base, bone2Name=dyn_skirt_bk_MSTR, swing1LimitDegrees=10, swing2LimitDegrees=20, twistLimitDegrees=0}
 {bone1Name=dyn_skirt_bk_01, bone2Name=dyn_skirt_bk_base, swing1LimitDegrees=10, swing2LimitDegrees=20, twistLimitDegrees=0}
 {bone1Name=dyn_skirt_bk_02, bone2Name=dyn_skirt_bk_01, swing1LimitDegrees=10, swing2LimitDegrees=20, twistLimitDegrees=0}
 {
  bone1Name=dyn_skirt_l_front2_MSTR
  bone2Name=thigh_l
  swing1LimitDegrees=45
  swing1Motion=Free
  swing2LimitDegrees=45
  swing2Motion=Free
  twistLimitDegrees=45
  twistMotion=Free
 }
 {
  bone1Name=dyn_skirt_l_front_MSTR
  bone2Name=thigh_l
  swing1LimitDegrees=45
  swing1Motion=Free
  swing2LimitDegrees=45
  swing2Motion=Free
  twistLimitDegrees=45
  twistMotion=Free
 }
 {
  bone1Name=dyn_skirt_l_side_MSTR
  bone2Name=thigh_l
  swing1LimitDegrees=45
  swing1Motion=Free
  swing2LimitDegrees=45
  swing2Motion=Free
  twistLimitDegrees=45
  twistMotion=Free
 }
 {
  bone1Name=dyn_skirt_r_front2_MSTR
  bone2Name=thigh_r
  swing1LimitDegrees=45
  swing1Motion=Free
  swing2LimitDegrees=45
  swing2Motion=Free
  twistLimitDegrees=45
  twistMotion=Free
 }
 {
  bone1Name=dyn_skirt_r_front_MSTR
  bone2Name=thigh_r
  swing1LimitDegrees=45
  swing1Motion=Free
  swing2LimitDegrees=45
  swing2Motion=Free
  twistLimitDegrees=45
  twistMotion=Free
 }
 {
  bone1Name=dyn_skirt_r_side_MSTR
  bone2Name=thigh_r
  swing1LimitDegrees=45
  swing1Motion=Free
  swing2LimitDegrees=45
  swing2Motion=Free
  twistLimitDegrees=45
  twistMotion=Free
 }
 {
  bone1Name=thigh_l
  bone2Name=pelvis
  swing1LimitDegrees=45
  swing1Motion=Free
  swing2LimitDegrees=45
  swing2Motion=Free
  twistLimitDegrees=45
  twistMotion=Free
 }
 {
  bone1Name=thigh_r
  bone2Name=pelvis
  swing1LimitDegrees=45
  swing1Motion=Free
  swing2LimitDegrees=45
  swing2Motion=Free
  twistLimitDegrees=45
  twistMotion=Free
 }
 {
  bone1Name=dyn_skirt_bk_MSTR
  bone2Name=pelvis
  swing1LimitDegrees=45
  swing1Motion=Free
  swing2LimitDegrees=45
  swing2Motion=Free
  twistLimitDegrees=45
  twistMotion=Free
 }
 {
  bone1Name=dyn_skirt_bk_l_MSTR
  bone2Name=pelvis
  swing1LimitDegrees=45
  swing1Motion=Free
  swing2LimitDegrees=45
  swing2Motion=Free
  twistLimitDegrees=45
  twistMotion=Free
 }
 {
  bone1Name=dyn_skirt_bk_r_MSTR
  bone2Name=pelvis
  swing1LimitDegrees=45
  swing1Motion=Free
  swing2LimitDegrees=45
  swing2Motion=Free
  twistLimitDegrees=45
  twistMotion=Free
 }
]
