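# Physics asset for the DistantEchoCastle_Style female medium body cape, named for skydiving, providing capsule bodies and joints for cape bone dynamics.
# Four left-side chains (A, B, D, E) each run from a tiny master capsule through two or three dynamic segments with narrow swing limits and zero twist limit.
# Chains A and B attach to the pelvis, D and E to spine_03.
# Pelvis, spine_03 and upperarm_r capsules supply the body shapes.
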PhysicsAsset /BRCosmetics/Characters/Player/Female/Medium/Bodies/F_MED_DistantEchoCastle_Style/Meshes/F_MED_DistantEchoCastle_Style_capeBBSkydive_Physics
bodies {
 CapeA_MSTR_l [{shapeType=Capsule, length=0.505, radius=0.505, shapeName=CapeA_MSTR_l_capsule, rotation=<rot 0 -90 0>}]
 CapeA_dyn_l_1 [{shapeType=Capsule, length=11.5, radius=2.75, shapeName=CapeA_dyn_l_1_capsule, center=<5 0 2.5>, rotation=<rot 90 0 0>}]
 CapeA_dyn_l_2 [{shapeType=Capsule, length=11, radius=3.25, shapeName=CapeA_dyn_l_2_capsule, center=<5 0 0>, rotation=<rot 90 0 0>}]
 CapeB_MSTR_l [{shapeType=Capsule, length=0.505, radius=0.505, shapeName=CapeB_MSTR_l_capsule, rotation=<rot 0 -90 0>}]
 CapeB_dyn_l_1 [{shapeType=Capsule, length=9.25, radius=5, shapeName=CapeB_dyn_l_1_capsule, center=<5 0 0>, rotation=<rot 90 0 0>}]
 CapeB_dyn_l_2 [{shapeType=Capsule, length=9.25, radius=5, shapeName=CapeB_dyn_l_2_capsule, center=<5 0 0>, rotation=<rot 90 0 0>}]
 CapeB_dyn_l_3 [{shapeType=Capsule, length=9.25, radius=7, shapeName=CapeB_dyn_l_3_capsule, center=<5 0 0>, rotation=<rot 90 0 0>}]
 CapeD_MSTR_l [{shapeType=Capsule, length=0.505, radius=0.505, shapeName=CapeD_MSTR_l_capsule, rotation=<rot 0 -90 0>}]
 CapeD_dyn_l_1 [{shapeType=Capsule, length=9.25, radius=5, shapeName=CapeD_dyn_l_1_capsule, center=<5 0 0>, rotation=<rot 90 0 0>}]
 CapeD_dyn_l_2 [{shapeType=Capsule, length=9.25, radius=5, shapeName=CapeD_dyn_l_2_capsule, center=<5 0 0>, rotation=<rot 90 0 0>}]
 CapeD_dyn_l_3 [{shapeType=Capsule, length=9.25, radius=5, shapeName=CapeD_dyn_l_3_capsule, center=<5 0 0>, rotation=<rot 90 0 0>}]
 CapeE_MSTR_l [{shapeType=Capsule, length=0.505, radius=0.505, shapeName=CapeE_MSTR_l_capsule, rotation=<rot 0 -90 0>}]
 CapeE_dyn_l_1 [{shapeType=Capsule, length=6.25, radius=8, shapeName=CapeE_dyn_l_1_capsule, center=<5 0 0>, rotation=<rot 90 0 0>}]
 CapeE_dyn_l_2 [{shapeType=Capsule, length=5.5, radius=8.75, shapeName=CapeE_dyn_l_2_capsule, center=<5 0 0>, rotation=<rot 90 0 0>}]
 pelvis [{shapeType=Capsule, length=21, radius=14.4661, shapeName=pelvis_capsule, center=<-15.9738 -2.7565 0>, rotation=<rot 90 -176.4237 163.5767>}]
 spine_03 [{shapeType=Capsule, length=4.75, radius=15.25, shapeName=spine_03_capsule, center=<-5.5118 4.2836 3.036>, rotation=<rot 0 -80.0008 90>}]
 upperarm_r [{shapeType=Capsule, length=25.1261, radius=2.1761, shapeName=upperarm_r_capsule, center=<-13.5704 -3.6675 0>, rotation=<rot 90 0 0>}]
}
constraints [
 {bone1Name=CapeA_dyn_l_1, bone2Name=CapeA_MSTR_l, swing1LimitDegrees=20, swing2LimitDegrees=20, twistLimitDegrees=0}
 {bone1Name=CapeA_dyn_l_2, bone2Name=CapeA_dyn_l_1, swing1LimitDegrees=20, swing2LimitDegrees=20, twistLimitDegrees=0}
 {bone1Name=CapeB_dyn_l_1, bone2Name=CapeB_MSTR_l, swing1LimitDegrees=10, swing2LimitDegrees=45, twistLimitDegrees=0}
 {bone1Name=CapeB_dyn_l_2, bone2Name=CapeB_dyn_l_1, swing1LimitDegrees=20, swing2LimitDegrees=20, twistLimitDegrees=0}
 {bone1Name=CapeB_dyn_l_3, bone2Name=CapeB_dyn_l_2, swing1LimitDegrees=20, swing2LimitDegrees=20, twistLimitDegrees=0}
 {bone1Name=CapeD_dyn_l_1, bone2Name=CapeD_MSTR_l, swing1LimitDegrees=30, swing2LimitDegrees=20, twistLimitDegrees=0}
 {bone1Name=CapeD_dyn_l_2, bone2Name=CapeD_dyn_l_1, swing1LimitDegrees=20, swing2LimitDegrees=20, twistLimitDegrees=0}
 {bone1Name=CapeD_dyn_l_3, bone2Name=CapeD_dyn_l_2, swing1LimitDegrees=20, swing2LimitDegrees=20, twistLimitDegrees=0}
 {bone1Name=CapeE_dyn_l_2, bone2Name=CapeE_dyn_l_1, swing1LimitDegrees=20, swing2LimitDegrees=20, twistLimitDegrees=0}
 {bone1Name=CapeE_dyn_l_1, bone2Name=CapeE_MSTR_l, swing1LimitDegrees=20, swing2LimitDegrees=20, twistLimitDegrees=0}
 {bone1Name=CapeE_MSTR_l, bone2Name=spine_03, swing1LimitDegrees=45, swing2LimitDegrees=45, twistLimitDegrees=45, twistMotion=Free}
 {
  bone1Name=CapeD_MSTR_l
  bone2Name=spine_03
  swing1LimitDegrees=45
  swing1Motion=Free
  swing2LimitDegrees=45
  swing2Motion=Free
  twistLimitDegrees=45
  twistMotion=Free
 }
 {
  bone1Name=CapeB_MSTR_l
  bone2Name=pelvis
  swing1LimitDegrees=45
  swing1Motion=Free
  swing2LimitDegrees=45
  swing2Motion=Free
  twistLimitDegrees=45
  twistMotion=Free
 }
 {
  bone1Name=CapeA_MSTR_l
  bone2Name=pelvis
  swing1LimitDegrees=45
  swing1Motion=Free
  swing2LimitDegrees=45
  swing2Motion=Free
  twistLimitDegrees=45
  twistMotion=Free
 }
 {
  bone1Name=upperarm_r
  bone2Name=spine_03
  swing1LimitDegrees=45
  swing1Motion=Free
  swing2LimitDegrees=45
  swing2Motion=Free
  twistLimitDegrees=45
  twistMotion=Free
 }
 {
  bone1Name=spine_03
  bone2Name=pelvis
  swing1LimitDegrees=45
  swing1Motion=Free
  swing2LimitDegrees=45
  swing2Motion=Free
  twistLimitDegrees=45
  twistMotion=Free
 }
]
