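# Animation Blueprint for the female DistantEchoCastle Style outfit, driving secondary motion on its cape, skirt, straps and belt extras.
# It copies the parent mesh pose, then layers Control Rigs, weighted bone constraints and RigidBody simulations.
# Cape branches built from a cached pose blend between standard and backbling physics setups; backblings also hide the canister.
# Gameplay and frontend state flags gate additive bone corrections for skydiving, movement, crouching, pickaxing and emotes.
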AnimBlueprint /BRCosmetics/Characters/Player/Female/Medium/Bodies/F_MED_DistantEchoCastle_Style/Meshes/F_MED_DistantEchoCastle_Style_AnimBP
target_skeleton=&/F_MED_DistantEchoCastle_Style_Skeleton
graphs [
 AnimGraph {nodes [
   n1 = Get bClothEnabled()
   Blend Poses by bool(BlendTime_0=0.2, BlendTime_1=0.5, bActiveValue=@n1) {blendType=Linear}
   n2 = "Transform (Modify) Bone - Bone: dyn_skirt_strapLong_MSTR"!(ComponentPose, bAlphaBoolEnabled:bool=@n1, Translation="0,0,0", Rotation="5,0,0", Scale="1,1,1") -> (Pose) {
    alphaInputType=Bool
    rotationMode=BMM_Additive
    rotationSpace=BCS_BoneSpace
    translationMode=BMM_Additive
    alphaBoolBlend {blendOption=Linear}
    alphaScaleBiasClamp {bInterpResult=true}
    boneToModify {boneName=dyn_skirt_strapLong_MSTR}
    rotation {pitch=5}
   }
   Component To Local(ComponentPose=@n2)
   Local To Component()
   Local To Component(LocalPose=Use cached pose 'Start_Cache'())
   n3 = Get CapeOn() -> (CapeOn:bool)
   "Transform (Modify) Bone - Bone: beltExtras4"!(ComponentPose, bAlphaBoolEnabled:bool=@n3, Translation="0,0,0", Rotation="0,0,0", Scale="0.75,0.75,1") -> (Pose) {
    alphaInputType=Bool
    rotationMode=BMM_Additive
    rotationSpace=BCS_BoneSpace
    scaleMode=BMM_Additive
    translationMode=BMM_Additive
    alphaBoolBlend {blendOption=Linear}
    alphaScaleBiasClamp {bInterpResult=true}
    boneToModify {boneName=beltExtras4}
    scale=<0.75 0.75>
   }
   n4 = "Constraint - clavicle_r_cape (RR)"!(ComponentPose=Local To Component(LocalPose=Use cached pose 'Start_Cache'()), Alpha=1, ConstraintWeights_0=1, ConstraintWeights_1:float=0) -> (Pose) {
    alphaBoolBlend {blendOption=Linear}
    boneToModify {boneName=clavicle_r_cape}
    constraintSetup [{transformType=Rotation, targetBone {boneName=clavicle_r}}, {transformType=Rotation, targetBone {boneName=upperarm_twist_01_r}}]
    constraintWeights [1,1]
   }
   n5 = "Constraint - CapeD_MSTR_l (RR)"!(ComponentPose=@n4, Alpha=1, ConstraintWeights_0=1, ConstraintWeights_1:float=0) -> (Pose) {
    alphaBoolBlend {blendOption=Linear}
    boneToModify {boneName=CapeD_MSTR_l}
    constraintSetup [{transformType=Rotation, targetBone {boneName=clavicle_r}}, {transformType=Rotation, targetBone {boneName=spine_05}}]
    constraintWeights [1,1]
   }
   n6 = "Constraint - CapeE_MSTR_l (RR)"!(ComponentPose=@n5, Alpha=1, ConstraintWeights_0=0, ConstraintWeights_1:float=1) -> (Pose) {
    alphaBoolBlend {blendOption=Linear}
    boneToModify {boneName=CapeE_MSTR_l}
    constraintSetup [{transformType=Rotation, targetBone {boneName=clavicle_r_cape}}, {transformType=Rotation, targetBone {boneName=spine_05}}]
    constraintWeights [1,1]
   }
   n7 = "Constraint - upperarm_twist_01_cape_r (RR)"!(ComponentPose=@n6, Alpha=1, ConstraintWeights_0=1, ConstraintWeights_1:float=0) -> (Pose) {
    alphaBoolBlend {blendOption=Linear}
    boneToModify {boneName=upperarm_twist_01_cape_r}
    constraintSetup [{transformType=Rotation, targetBone {boneName=upperarm_twist_01_r}}, {transformType=Rotation, targetBone {boneName=clavicle_l}}]
    constraintWeights [1,1]
   }
   n8 = RigidBody!(
    ComponentPose=@n7
    bAlphaBoolEnabled:bool=NOT Boolean(A=Get bIsSkyDiving())
    OverrideWorldGravity:Vector=Get GravityOverride()
    ExternalForce
   ) -> (Pose) {
    alphaInputType=Bool
    bClampLinearTranslationLimitToRefPose=true
    bOverrideWorldGravity=true
    overridePhysicsAsset=&/F_MED_DistantEchoCastle_Style_capeBB_Physics
    simulationSpace=BaseBoneSpace
    alphaBoolBlend {blendOption=Linear}
    baseBoneRef {boneName=root}
    componentAppliedLinearAccClamp=<5000 5000 5000>
    componentLinearAccScale=<0.6 0.5 0.5>
    componentLinearVelScale=<0.6 0.5 0.5>
    overrideWorldGravity {z=-980}
    simSpaceSettings {
     maxAngularAcceleration=500
     maxAngularVelocity=3
     maxLinearAcceleration=700
     maxLinearVelocity=500
     velocityScaleZ=0
     worldAlpha=0.25
     externalLinearDragV=<1 1>
    }
   }
   n9 = RigidBody!(ComponentPose=@n8, Alpha:float=Get SkydiveCapeRB() -> (SkydiveCapeRB:double), OverrideWorldGravity="0,0,0", ExternalForce) -> (Pose) {
    bClampLinearTranslationLimitToRefPose=true
    bOverrideWorldGravity=true
    overridePhysicsAsset=&/F_MED_DistantEchoCastle_Style_capeBBSkydive_Physics
    simulationSpace=BaseBoneSpace
    alphaBoolBlend {blendOption=Linear}
    baseBoneRef {boneName=root}
    componentAppliedLinearAccClamp=<5000 5000 5000>
    componentLinearAccScale=<0.6 0.5 0.5>
    componentLinearVelScale=<0.6 0.5 0.5>
    overrideWorldGravity {z=-980}
    simSpaceSettings {
     maxAngularAcceleration=500
     maxAngularVelocity=3
     maxLinearAcceleration=700
     maxLinearVelocity=500
     velocityScaleZ=0
     worldAlpha=0.25
     externalLinearDragV=<1 1>
    }
   }
   n10 = "Transform (Modify) Bone - Bone: beltExtras2"!(ComponentPose=@n9, bAlphaBoolEnabled:bool=True, Translation="0,0,0", Rotation="0,0,0", Scale="0,0,0") -> (Pose) {
    alphaInputType=Bool
    rotationMode=BMM_Additive
    rotationSpace=BCS_BoneSpace
    scaleMode=BMM_Additive
    translationMode=BMM_Additive
    alphaBoolBlend {blendOption=Linear}
    alphaScaleBiasClamp {bInterpResult=true}
    boneToModify {boneName=beltExtras2}
    scale=<0 0 0>
   }
   n11 = Get isWearingBackbling()
   n12 = "Transform (Modify) Bone - Bone: beltExtras3"!(ComponentPose=@n10, bAlphaBoolEnabled:bool=@n11, Translation="0,0,0", Rotation="0,0,0", Scale="0,0,0") -> (Pose) {
    alphaInputType=Bool
    rotationMode=BMM_Additive
    rotationSpace=BCS_BoneSpace
    scaleMode=BMM_Additive
    translationMode=BMM_Additive
    alphaBoolBlend {blendOption=Linear}
    alphaScaleBiasClamp {bInterpResult=true}
    boneToModify {boneName=beltExtras3}
    scale=<0 0 0>
   }
   n13 = "Transform (Modify) Bone - Bone: beltExtras4"(ComponentPose=@n12, bAlphaBoolEnabled=@n11, Translation="0,0,0", Rotation="0,0,0", Scale="0,0,0") {
    alphaInputType=Bool
    rotationMode=BMM_Additive
    rotationSpace=BCS_BoneSpace
    scaleMode=BMM_Additive
    translationMode=BMM_Additive
    alphaBoolBlend {blendOption=Linear}
    alphaScaleBiasClamp {bInterpResult=true}
    boneToModify {boneName=beltExtras4}
    scale=<0 0 0>
   }
   n14 = "Transform (Modify) Bone - Bone: beltExtras5"!(ComponentPose=@n13, bAlphaBoolEnabled:bool=@n11, Translation="0,0,0", Rotation="0,0,0", Scale="0,0,0") -> (Pose) {
    alphaInputType=Bool
    rotationMode=BMM_Additive
    rotationSpace=BCS_BoneSpace
    scaleMode=BMM_Additive
    translationMode=BMM_Additive
    alphaBoolBlend {blendOption=Linear}
    alphaScaleBiasClamp {bInterpResult=true}
    boneToModify {boneName=beltExtras5}
    scale=<0 0 0>
   }
   n15 = "Transform (Modify) Bone - Bone: CapeA_MSTR_l"!(ComponentPose=@n14, bAlphaBoolEnabled:bool=True, Translation="0,0,0", Rotation="10,-15,0", Scale="1,1,1") -> (Pose) {
    alphaInputType=Bool
    rotationMode=BMM_Additive
    rotationSpace=BCS_BoneSpace
    scaleMode=BMM_Additive
    translationMode=BMM_Additive
    alphaBoolBlend {blendOption=Linear}
    alphaScaleBiasClamp {bInterpResult=true}
    boneToModify {boneName=CapeA_MSTR_l}
    rotation {pitch=10, yaw=-15}
   }
   n16 = Get bIsSkyDiving()
   n17 = "Transform (Modify) Bone - Bone: CapeA_dyn_l_1"!(ComponentPose=@n15, bAlphaBoolEnabled:bool=@n16, Translation="0,0,0", Rotation="10,0,0", Scale="1,1,1") -> (Pose) {
    alphaInputType=Bool
    rotationMode=BMM_Additive
    rotationSpace=BCS_BoneSpace
    translationMode=BMM_Additive
    alphaBoolBlend {blendOption=Linear}
    alphaScaleBiasClamp {bInterpResult=true}
    boneToModify {boneName=CapeA_dyn_l_1}
    rotation {pitch=10}
   }
   n18 = "Transform (Modify) Bone - Bone: CapeB_MSTR_l"!(ComponentPose=@n17, bAlphaBoolEnabled:bool=@n16, Translation="0,5,0", Rotation="20,0,0", Scale="1,1,1") -> (Pose) {
    alphaInputType=Bool
    rotationMode=BMM_Additive
    rotationSpace=BCS_BoneSpace
    translationMode=BMM_Additive
    alphaBoolBlend {blendOption=Linear}
    alphaScaleBiasClamp {bInterpResult=true}
    boneToModify {boneName=CapeB_MSTR_l}
    rotation {pitch=20}
    translation {y=5}
   }
   n19 = "Transform (Modify) Bone - Bone: CapeD_MSTR_l"!(ComponentPose=@n18, bAlphaBoolEnabled:bool=@n16, Translation="0,5,0", Rotation="20,0,0", Scale="1,1,1") -> (Pose) {
    alphaInputType=Bool
    rotationMode=BMM_Additive
    rotationSpace=BCS_BoneSpace
    translationMode=BMM_Additive
    alphaBoolBlend {blendOption=Linear}
    alphaScaleBiasClamp {bInterpResult=true}
    boneToModify {boneName=CapeD_MSTR_l}
    rotation {pitch=20}
    translation {y=5}
   }
   n20 = "Transform (Modify) Bone - Bone: CapeE_dyn_l_2"!(ComponentPose=@n19, bAlphaBoolEnabled:bool=@n16, Translation="0,0,0", Rotation="20,0,0", Scale="1,1,1") -> (Pose) {
    alphaInputType=Bool
    rotationMode=BMM_Additive
    rotationSpace=BCS_BoneSpace
    translationMode=BMM_Additive
    alphaBoolBlend {blendOption=Linear}
    alphaScaleBiasClamp {bInterpResult=true}
    boneToModify {boneName=CapeE_dyn_l_2}
    rotation {pitch=20}
   }
   n21 = Component To Local(ComponentPose=@n20)
   n22 = "Constraint - clavicle_r_cape (RR)"(ComponentPose=Local To Component(LocalPose=Use cached pose 'Start_Cache'()), Alpha=1, ConstraintWeights_0=1, ConstraintWeights_1=0) {
    alphaBoolBlend {blendOption=Linear}
    boneToModify {boneName=clavicle_r_cape}
    constraintSetup [{transformType=Rotation, targetBone {boneName=clavicle_r}}, {transformType=Rotation, targetBone {boneName=upperarm_twist_01_r}}]
    constraintWeights [1,1]
   }
   n23 = "Constraint - CapeD_MSTR_l (RR)"(ComponentPose=@n22, Alpha=1, ConstraintWeights_0=1, ConstraintWeights_1=0) {
    alphaBoolBlend {blendOption=Linear}
    boneToModify {boneName=CapeD_MSTR_l}
    constraintSetup [{transformType=Rotation, targetBone {boneName=clavicle_r}}, {transformType=Rotation, targetBone {boneName=spine_05}}]
    constraintWeights [1,1]
   }
   n24 = "Constraint - CapeE_MSTR_l (RR)"(ComponentPose=@n23, Alpha=1, ConstraintWeights_0=0, ConstraintWeights_1=1) {
    alphaBoolBlend {blendOption=Linear}
    boneToModify {boneName=CapeE_MSTR_l}
    constraintSetup [{transformType=Rotation, targetBone {boneName=clavicle_r_cape}}, {transformType=Rotation, targetBone {boneName=spine_05}}]
    constraintWeights [1,1]
   }
   n25 = "Constraint - upperarm_twist_01_cape_r (RR)"(ComponentPose=@n24, Alpha=1, ConstraintWeights_0=1, ConstraintWeights_1=0) {
    alphaBoolBlend {blendOption=Linear}
    boneToModify {boneName=upperarm_twist_01_cape_r}
    constraintSetup [{transformType=Rotation, targetBone {boneName=upperarm_twist_01_r}}, {transformType=Rotation, targetBone {boneName=clavicle_l}}]
    constraintWeights [1,1]
   }
   n26 = Get bIsSkyDiving()
   n27 = NOT Boolean(A=@n26)
   n28 = "Transform (Modify) Bone - Bone: CapeE_MSTR_l"!(ComponentPose=@n25, bAlphaBoolEnabled:bool=@n27, Translation="0,0,0", Rotation="0,0,0", Scale="1,1,1") -> (Pose) {
    alphaInputType=Bool
    rotationMode=BMM_Additive
    rotationSpace=BCS_BoneSpace
    translationMode=BMM_Additive
    alphaBoolBlend {blendOption=Linear}
    alphaScaleBiasClamp {bInterpResult=true}
    boneToModify {boneName=CapeE_MSTR_l}
   }
   n29 = "Transform (Modify) Bone - Bone: CapeD_MSTR_l"(ComponentPose=@n28, bAlphaBoolEnabled=@n27, Translation="5,-5,0", Rotation="0,0,0", Scale="1,1,1") {
    alphaInputType=Bool
    rotationMode=BMM_Additive
    rotationSpace=BCS_BoneSpace
    translationMode=BMM_Additive
    alphaBoolBlend {blendOption=Linear}
    alphaScaleBiasClamp {bInterpResult=true}
    boneToModify {boneName=CapeD_MSTR_l}
    translation=<5 -5>
   }
   n30 = "Transform (Modify) Bone - Bone: CapeB_MSTR_l"(ComponentPose=@n29, bAlphaBoolEnabled=@n27, Translation="0,0,3", Rotation="0,0,0", Scale="1,1,1") {
    alphaInputType=Bool
    rotationMode=BMM_Additive
    rotationSpace=BCS_BoneSpace
    translationMode=BMM_Additive
    translationSpace=BCS_BoneSpace
    alphaBoolBlend {blendOption=Linear}
    alphaScaleBiasClamp {bInterpResult=true}
    boneToModify {boneName=CapeB_MSTR_l}
    translation {z=3}
   }
   n31 = "Transform (Modify) Bone - Bone: CapeA_MSTR_l"(ComponentPose=@n30, bAlphaBoolEnabled=@n27, Translation="3,-3,0", Rotation="0,0,0", Scale="1,1,1") {
    alphaInputType=Bool
    rotationMode=BMM_Additive
    rotationSpace=BCS_BoneSpace
    translationMode=BMM_Additive
    alphaBoolBlend {blendOption=Linear}
    alphaScaleBiasClamp {bInterpResult=true}
    boneToModify {boneName=CapeA_MSTR_l}
    translation=<3 -3>
   }
   n32 = "Transform (Modify) Bone - Bone: CapeE_MSTR_l"(ComponentPose=@n31, bAlphaBoolEnabled=Get bIsTacticalSprinting(), Translation="8,0,5", Rotation="0,0,0", Scale="1,1,1") {
    alphaInputType=Bool
    rotationMode=BMM_Additive
    rotationSpace=BCS_BoneSpace
    translationMode=BMM_Additive
    alphaBoolBlend {blendOption=Linear}
    alphaScaleBiasClamp {bInterpResult=true}
    boneToModify {boneName=CapeE_MSTR_l}
    translation {x=8, z=5}
   }
   n33 = "Transform (Modify) Bone - Bone: CapeE_MSTR_l"(ComponentPose=@n32, bAlphaBoolEnabled=@n26, Translation="0,0,0", Rotation="40,0,0", Scale="1,1,1") {
    alphaInputType=Bool
    rotationMode=BMM_Additive
    rotationSpace=BCS_BoneSpace
    translationMode=BMM_Additive
    alphaBoolBlend {blendOption=Linear}
    alphaScaleBiasClamp {bInterpResult=true}
    boneToModify {boneName=CapeE_MSTR_l}
    rotation {pitch=40}
   }
   n34 = "Transform (Modify) Bone - Bone: CapeD_MSTR_l"(ComponentPose=@n33, bAlphaBoolEnabled=@n26, Translation="0,0,0", Rotation="40,0,0", Scale="1,1,1") {
    alphaInputType=Bool
    rotationMode=BMM_Additive
    rotationSpace=BCS_BoneSpace
    translationMode=BMM_Additive
    alphaBoolBlend {blendOption=Linear}
    alphaScaleBiasClamp {bInterpResult=true}
    boneToModify {boneName=CapeD_MSTR_l}
    rotation {pitch=40}
   }
   n35 = "Transform (Modify) Bone - Bone: CapeB_MSTR_l"(ComponentPose=@n34, bAlphaBoolEnabled=@n26, Translation="0,0,0", Rotation="40,0,0", Scale="1,1,1") {
    alphaInputType=Bool
    rotationMode=BMM_Additive
    rotationSpace=BCS_BoneSpace
    translationMode=BMM_Additive
    alphaBoolBlend {blendOption=Linear}
    alphaScaleBiasClamp {bInterpResult=true}
    boneToModify {boneName=CapeB_MSTR_l}
    rotation {pitch=40}
   }
   n36 = "Transform (Modify) Bone - Bone: CapeA_MSTR_l"(ComponentPose=@n35, bAlphaBoolEnabled=@n26, Translation="0,0,0", Rotation="40,0,0", Scale="1,1,1") {
    alphaInputType=Bool
    rotationMode=BMM_Additive
    rotationSpace=BCS_BoneSpace
    translationMode=BMM_Additive
    alphaBoolBlend {blendOption=Linear}
    alphaScaleBiasClamp {bInterpResult=true}
    boneToModify {boneName=CapeA_MSTR_l}
    rotation {pitch=40}
   }
   n37 = "Transform (Modify) Bone - Bone: CapeE_MSTR_l"(ComponentPose=@n36, bAlphaBoolEnabled=@n26, Translation="0,0,0", Rotation="40,0,0", Scale="1,1,1") {
    alphaInputType=Bool
    rotationMode=BMM_Additive
    rotationSpace=BCS_BoneSpace
    translationMode=BMM_Additive
    alphaBoolBlend {blendOption=Linear}
    alphaScaleBiasClamp {bInterpResult=true}
    boneToModify {boneName=CapeE_MSTR_l}
    rotation {pitch=40}
   }
   n38 = RigidBody!(ComponentPose=@n37, bAlphaBoolEnabled:bool, OverrideWorldGravity=Get GravityOverride(), ExternalForce) -> (Pose) {
    alphaInputType=Bool
    bClampLinearTranslationLimitToRefPose=true
    bOverrideWorldGravity=true
    overridePhysicsAsset=&/F_MED_DistantEchoCastle_Style_cape_Physics
    simulationSpace=BaseBoneSpace
    alphaBoolBlend {blendOption=Linear}
    baseBoneRef {boneName=root}
    componentAppliedLinearAccClamp=<5000 5000 5000>
    componentLinearAccScale=<0.6 0.5 0.5>
    componentLinearVelScale=<0.6 0.5 0.5>
    overrideWorldGravity {z=-980}
    simSpaceSettings {
     maxAngularAcceleration=500
     maxAngularVelocity=3
     maxLinearAcceleration=700
     maxLinearVelocity=500
     velocityScaleZ=0
     worldAlpha=0.25
     externalLinearDragV=<1 1>
    }
   }
   n39 = "Transform (Modify) Bone - Bone: CapeB_MSTR_l"(ComponentPose=@n38, bAlphaBoolEnabled=@n27, Translation="0,-3,0", Rotation="0,0,0", Scale="1,1,1") {
    alphaInputType=Bool
    rotationMode=BMM_Additive
    rotationSpace=BCS_BoneSpace
    translationMode=BMM_Additive
    alphaBoolBlend {blendOption=Linear}
    alphaScaleBiasClamp {bInterpResult=true}
    boneToModify {boneName=CapeB_MSTR_l}
    translation {y=-3}
   }
   n40 = "Transform (Modify) Bone - Bone: CapeE_MSTR_l"!(ComponentPose=@n39, Alpha:float=Get BwdMovement() -> (BwdMovement:double), Translation="0,5,5", Rotation="0,0,0", Scale="1,1,1") -> (Pose) {
    rotationMode=BMM_Additive
    rotationSpace=BCS_BoneSpace
    translationMode=BMM_Additive
    translationSpace=BCS_BoneSpace
    alphaBoolBlend {blendOption=Linear}
    alphaScaleBiasClamp {bInterpResult=true}
    boneToModify {boneName=CapeE_MSTR_l}
    translation {y=5, z=5}
   }
   n41 = "Transform (Modify) Bone - Bone: beltExtras2"(ComponentPose=@n40, bAlphaBoolEnabled=@n3, Translation="0,0,0", Rotation="0,0,0", Scale="0.75,0.75,1") {
    alphaInputType=Bool
    rotationMode=BMM_Additive
    rotationSpace=BCS_BoneSpace
    scaleMode=BMM_Additive
    translationMode=BMM_Additive
    alphaBoolBlend {blendOption=Linear}
    alphaScaleBiasClamp {bInterpResult=true}
    boneToModify {boneName=beltExtras2}
    scale=<0.75 0.75>
   }
   n42 = "Transform (Modify) Bone - Bone: beltExtras3"(ComponentPose=@n41, bAlphaBoolEnabled=@n3, Translation="0,0,0", Rotation="0,0,0", Scale="0.75,0.75,1") {
    alphaInputType=Bool
    rotationMode=BMM_Additive
    rotationSpace=BCS_BoneSpace
    scaleMode=BMM_Additive
    translationMode=BMM_Additive
    alphaBoolBlend {blendOption=Linear}
    alphaScaleBiasClamp {bInterpResult=true}
    boneToModify {boneName=beltExtras3}
    scale=<0.75 0.75>
   }
   n43 = "Transform (Modify) Bone - Bone: beltExtras5"(ComponentPose=@n42, bAlphaBoolEnabled=@n3, Translation="0,0,0", Rotation="0,0,0", Scale="0.75,0.75,1") {
    alphaInputType=Bool
    rotationMode=BMM_Additive
    rotationSpace=BCS_BoneSpace
    scaleMode=BMM_Additive
    translationMode=BMM_Additive
    alphaBoolBlend {blendOption=Linear}
    alphaScaleBiasClamp {bInterpResult=true}
    boneToModify {boneName=beltExtras5}
    scale=<0.75 0.75>
   }
   n44 = Component To Local(ComponentPose=@n43)
   n45 = AND Boolean(A=Get isWearingBackbling(), B=NOT Boolean(A=Get isWearingCapeBool() -> (isWearingCapeBool:bool)))
   n46 = "Transform (Modify) Bone - Bone: clavicle_r_cape"!(
    ComponentPose=Local To Component(LocalPose=Blend Poses by bool(BlendPose_0=@n21, BlendPose_1=@n44, BlendTime_0=0.3, BlendTime_1=0.3, bActiveValue=@n45) {blendType=Linear})
    bAlphaBoolEnabled:bool=Get bIsSkyDiving()
    Translation="0,0,0"
    Rotation="0,0,0"
    Scale="1,2,1"
   ) -> (Pose) {
    alphaInputType=Bool
    rotationMode=BMM_Additive
    rotationSpace=BCS_BoneSpace
    scaleMode=BMM_Additive
    translationMode=BMM_Additive
    alphaBoolBlend {blendOption=Linear}
    alphaScaleBiasClamp {bInterpResult=true}
    boneToModify {boneName=clavicle_r_cape}
    scale {y=2}
   }
   n47 = Get CapeOnInFrontend() -> (CapeOnInFrontend:bool)
   n48 = "Transform (Modify) Bone - Bone: CapeA_MSTR_l"(ComponentPose=@n46, bAlphaBoolEnabled=@n47, Translation="0,0,0", Rotation="0,-20,0", Scale="1,1,1") {
    alphaInputType=Bool
    rotationMode=BMM_Additive
    rotationSpace=BCS_BoneSpace
    scaleMode=BMM_Additive
    translationMode=BMM_Additive
    alphaBoolBlend {blendOption=Linear}
    alphaScaleBiasClamp {bInterpResult=true}
    boneToModify {boneName=CapeA_MSTR_l}
    rotation {yaw=-20}
   }
   n49 = Get frontendNoEmote() -> (frontendNoEmote:bool)
   n50 = "Transform (Modify) Bone - Bone: dyn_skirt_r_front2_MSTR"!(ComponentPose=@n48, bAlphaBoolEnabled:bool=@n49, Translation="0,5,0", Rotation="-20,0,20", Scale="1,1,1") -> (Pose) {
    alphaInputType=Bool
    rotationMode=BMM_Additive
    rotationSpace=BCS_BoneSpace
    scaleMode=BMM_Additive
    translationMode=BMM_Additive
    alphaBoolBlend {blendOption=Linear}
    alphaScaleBiasClamp {bInterpResult=true}
    boneToModify {boneName=dyn_skirt_r_front2_MSTR}
    rotation {pitch=-20, roll=20}
    translation {y=5}
   }
   n51 = "Transform (Modify) Bone - Bone: dyn_skirt_l_front2_MSTR"!(ComponentPose=@n50, bAlphaBoolEnabled:bool=@n49, Translation="-3,0,0", Rotation="0,0,0", Scale="1,1,1") -> (Pose) {
    alphaInputType=Bool
    rotationMode=BMM_Additive
    rotationSpace=BCS_BoneSpace
    scaleMode=BMM_Additive
    translationMode=BMM_Additive
    alphaBoolBlend {blendOption=Linear}
    alphaScaleBiasClamp {bInterpResult=true}
    boneToModify {boneName=dyn_skirt_l_front2_MSTR}
    translation {x=-3}
   }
   n52 = "Transform (Modify) Bone - Bone: CapeD_MSTR_l"(ComponentPose=@n51, bAlphaBoolEnabled=@n47, Translation="0,-5,0", Rotation="-20,0,0", Scale="1,1,1") {
    alphaInputType=Bool
    rotationMode=BMM_Additive
    rotationSpace=BCS_BoneSpace
    translationMode=BMM_Additive
    alphaBoolBlend {blendOption=Linear}
    alphaScaleBiasClamp {bInterpResult=true}
    boneToModify {boneName=CapeD_MSTR_l}
    rotation {pitch=-20}
    translation {y=-5}
   }
   n53 = "Transform (Modify) Bone - Bone: dyn_skirt_strapLong_MSTR"(ComponentPose=@n52, bAlphaBoolEnabled=@n47, Translation="0,0,0", Rotation="-8,0,0", Scale="1,1,1") {
    alphaInputType=Bool
    rotationMode=BMM_Additive
    rotationSpace=BCS_BoneSpace
    translationMode=BMM_Additive
    alphaBoolBlend {blendOption=Linear}
    alphaScaleBiasClamp {bInterpResult=true}
    boneToModify {boneName=dyn_skirt_strapLong_MSTR}
    rotation {pitch=-8}
   }
   n54 = "Transform (Modify) Bone - Bone: CapeA_MSTR_l"(
    ComponentPose=@n53
    bAlphaBoolEnabled=Get BackblingCapeFrontend() -> (BackblingCapeFrontend:bool)
    Translation="0,0,0"
    Rotation="0,25,0"
    Scale="1,1,1"
   ) {
    alphaInputType=Bool
    rotationMode=BMM_Additive
    rotationSpace=BCS_BoneSpace
    translationMode=BMM_Additive
    alphaBoolBlend {blendOption=Linear}
    alphaScaleBiasClamp {bInterpResult=true}
    boneToModify {boneName=CapeA_MSTR_l}
    rotation {yaw=25}
   }
   n55 = Get RigidBodiesOffInFrontend() -> (RigidBodiesOffInFrontend:bool)
   n56 = "Transform (Modify) Bone - Bone: dyn_skirt_l_front_MSTR"!(ComponentPose=@n54, bAlphaBoolEnabled:bool=@n55, Translation="0,0,0", Rotation="5,5,0", Scale="1,1,1") -> (Pose) {
    alphaInputType=Bool
    rotationMode=BMM_Additive
    rotationSpace=BCS_BoneSpace
    scaleMode=BMM_Additive
    translationMode=BMM_Additive
    alphaBoolBlend {blendOption=Linear}
    alphaScaleBiasClamp {bInterpResult=true}
    boneToModify {boneName=dyn_skirt_l_front_MSTR}
    rotation {pitch=5, yaw=5}
   }
   n57 = "Transform (Modify) Bone - Bone: dyn_skirt_r_front2_MSTR"(ComponentPose=@n56, bAlphaBoolEnabled=@n55, Translation="0,0,0", Rotation="5,5,0", Scale="1,1,1") {
    alphaInputType=Bool
    rotationMode=BMM_Additive
    rotationSpace=BCS_BoneSpace
    scaleMode=BMM_Additive
    translationMode=BMM_Additive
    alphaBoolBlend {blendOption=Linear}
    alphaScaleBiasClamp {bInterpResult=true}
    boneToModify {boneName=dyn_skirt_r_front2_MSTR}
    rotation {pitch=5, yaw=5}
   }
   n58 = Get isPickaxing() -> (isPickaxing:float)
   n59 = "Transform (Modify) Bone - Bone: CapeD_MSTR_l"!(ComponentPose=@n57, Alpha:float=@n58, Translation="0,-5,0", Rotation="25,0,0", Scale="1,1,1") -> (Pose) {
    rotationMode=BMM_Additive
    rotationSpace=BCS_BoneSpace
    translationMode=BMM_Additive
    alphaBoolBlend {blendOption=Linear}
    alphaScaleBiasClamp {bInterpResult=true, interpSpeedDecreasing=5}
    boneToModify {boneName=CapeD_MSTR_l}
    rotation {pitch=25}
    translation {y=-5}
   }
   n60 = "Transform (Modify) Bone - Bone: CapeE_MSTR_l"(ComponentPose=@n59, Alpha=@n58, Translation="0,-5,0", Rotation="25,0,0", Scale="1,1,1") {
    rotationMode=BMM_Additive
    rotationSpace=BCS_BoneSpace
    translationMode=BMM_Additive
    alphaBoolBlend {blendOption=Linear}
    alphaScaleBiasClamp {bInterpResult=true, interpSpeedDecreasing=5}
    boneToModify {boneName=CapeE_MSTR_l}
    rotation {pitch=25}
    translation {y=-5}
   }
   n61 = "Transform (Modify) Bone - Bone: CapeB_MSTR_l"!(ComponentPose=@n60, Alpha:float=@n58, Translation="0,-5,0", Rotation="25,0,0", Scale="1,1,1") -> (Pose) {
    rotationMode=BMM_Additive
    rotationSpace=BCS_BoneSpace
    translationMode=BMM_Additive
    alphaBoolBlend {blendOption=Linear}
    alphaScaleBiasClamp {bInterpResult=true, interpSpeedDecreasing=5}
    boneToModify {boneName=CapeB_MSTR_l}
    rotation {pitch=25}
    translation {y=-5}
   }
   n62 = "Transform (Modify) Bone - Bone: Cannister"!(ComponentPose=@n61, bAlphaBoolEnabled:bool=Get isWearingBackbling(), Translation="0,0,0", Scale="0,0,0") -> (Pose) {
    alphaInputType=Bool
    scaleMode=BMM_Replace
    scaleSpace=BCS_BoneSpace
    translationMode=BMM_Additive
    translationSpace=BCS_BoneSpace
    boneToModify {boneName=Cannister}
    scale=<0 0 0>
   }
   Output Pose(Result=Component To Local(ComponentPose=@n62)) {layerGroup=None}
   n63 = Control Rig(Source=Copy Pose From Mesh() {bUseAttachedParent=true}) {
    alphaBoolBlend {blendOption=Linear}
    controlRigAssetReference {blueprintRigClass=/Game/Characters/Player/Female/Medium/Bodies/F_MED_DistantEchoCastle_Style/Meshes/F_MED_DeformRig_DistantEchoCastleStyle_ControlRig_new1.F_MED_DeformRig_DistantEchoCastleStyle_ControlRig_new1_C}
   }
   n64 = "Constraint - beltExtras5 (PPPPPPPPPP)"!(
    ComponentPose=Local To Component(LocalPose=@n63)
    Alpha=1
    ConstraintWeights_0=0.0183
    ConstraintWeights_1:float=0.0194
    ConstraintWeights_2:float=0.1571
    ConstraintWeights_3:float=0.0028
    ConstraintWeights_4:float=0.0704
    ConstraintWeights_5:float=0.1924
    ConstraintWeights_6:float=0.0448
    ConstraintWeights_7:float=0.4759
    ConstraintWeights_8:float=0.0034
    ConstraintWeights_9:float=0.0155
   ) -> (Pose) {
    alphaBoolBlend {blendOption=Linear}
    boneToModify {boneName=beltExtras5}
    constraintSetup [
     {targetBone {boneName=spine_05}}
     {targetBone {boneName=clavicle_r}}
     {targetBone {boneName=spine_04}}
     {targetBone {boneName=upperarm_twist_01_r}}
     {targetBone {boneName=spine_03}}
     {targetBone {boneName=spine_02}}
     {targetBone {boneName=spine_01}}
     {targetBone {boneName=pelvis}}
     {targetBone {boneName=thigh_twist_01_r}}
     {targetBone {boneName=deform_glute_r}}
    ]
    constraintWeights [1,1,1,1,1,1,1,1,1,1]
   }
   n65 = "Constraint - beltExtras4 (PPPPPPPPPP)"!(
    ComponentPose=@n64
    Alpha=1
    ConstraintWeights_0=0.0212
    ConstraintWeights_1:float=0.0006
    ConstraintWeights_2:float=0.0185
    ConstraintWeights_3:float=0.1319
    ConstraintWeights_4:float=0.0006
    ConstraintWeights_5:float=0.0798
    ConstraintWeights_6:float=0.2169
    ConstraintWeights_7:float=0.0604
    ConstraintWeights_8:float=0.4616
    ConstraintWeights_9:float=0.0084
   ) -> (Pose) {
    alphaBoolBlend {blendOption=Linear}
    boneToModify {boneName=beltExtras4}
    constraintSetup [
     {targetBone {boneName=spine_05}}
     {targetBone {boneName=clavicle_l}}
     {targetBone {boneName=clavicle_r}}
     {targetBone {boneName=spine_04}}
     {targetBone {boneName=upperarm_twist_01_r}}
     {targetBone {boneName=spine_03}}
     {targetBone {boneName=spine_02}}
     {targetBone {boneName=spine_01}}
     {targetBone {boneName=pelvis}}
     {targetBone {boneName=deform_glute_r}}
    ]
    constraintWeights [1,1,1,1,1,1,1,1,1,1]
   }
   n66 = "Constraint - beltExtras3 (PPPPPPPPP)"!(
    ComponentPose=@n65
    Alpha=1
    ConstraintWeights_0=0.0255
    ConstraintWeights_1:float=0.0088
    ConstraintWeights_2:float=0.0143
    ConstraintWeights_3:float=0.1276
    ConstraintWeights_4:float=0.0868
    ConstraintWeights_5:float=0.2483
    ConstraintWeights_6:float=0.0778
    ConstraintWeights_7:float=0.4102
    ConstraintWeights_8:float=0.0006
   ) -> (Pose) {
    alphaBoolBlend {blendOption=Linear}
    boneToModify {boneName=beltExtras3}
    constraintSetup [
     {targetBone {boneName=spine_05}}
     {targetBone {boneName=clavicle_l}}
     {targetBone {boneName=clavicle_r}}
     {targetBone {boneName=spine_04}}
     {targetBone {boneName=spine_03}}
     {targetBone {boneName=spine_02}}
     {targetBone {boneName=spine_01}}
     {targetBone {boneName=pelvis}}
     {targetBone {boneName=deform_glute_r}}
    ]
    constraintWeights [1,1,1,1,1,1,1,1,1]
   }
   n67 = "Constraint - beltExtras2 (PPPPPPPPPP)"!(
    ComponentPose=@n66
    Alpha=1
    ConstraintWeights_0=0.037
    ConstraintWeights_1:float=0.03
    ConstraintWeights_2:float=0.0045
    ConstraintWeights_3:float=0.192
    ConstraintWeights_4:float=0.002
    ConstraintWeights_5:float=0.1119
    ConstraintWeights_6:float=0.2965
    ConstraintWeights_7:float=0.081
    ConstraintWeights_8:float=0.2436
    ConstraintWeights_9:float=0.0015
   ) -> (Pose) {
    alphaBoolBlend {blendOption=Linear}
    boneToModify {boneName=beltExtras2}
    constraintSetup [
     {targetBone {boneName=spine_05}}
     {targetBone {boneName=clavicle_l}}
     {targetBone {boneName=clavicle_r}}
     {targetBone {boneName=spine_04}}
     {targetBone {boneName=upperarm_twist_01_l}}
     {targetBone {boneName=spine_03}}
     {targetBone {boneName=spine_02}}
     {targetBone {boneName=spine_01}}
     {targetBone {boneName=pelvis}}
     {targetBone {boneName=deform_glute_l}}
    ]
    constraintWeights [1,1,1,1,1,1,1,1,1,1]
   }
   n68 = "Constraint - beltExtras1 (PPPPPPPPPP)"!(
    ComponentPose=@n67
    Alpha=1
    ConstraintWeights_0=0.0204
    ConstraintWeights_1:float=0.0212
    ConstraintWeights_2:float=0.1579
    ConstraintWeights_3:float=0.0029
    ConstraintWeights_4:float=0.0754
    ConstraintWeights_5:float=0.2261
    ConstraintWeights_6:float=0.0563
    ConstraintWeights_7:float=0.4385
    ConstraintWeights_8:float=0.0002
    ConstraintWeights_9:float=0.0011
   ) -> (Pose) {
    alphaBoolBlend {blendOption=Linear}
    boneToModify {boneName=beltExtras1}
    constraintSetup [
     {targetBone {boneName=spine_05}}
     {targetBone {boneName=clavicle_l}}
     {targetBone {boneName=spine_04}}
     {targetBone {boneName=upperarm_twist_01_l}}
     {targetBone {boneName=spine_03}}
     {targetBone {boneName=spine_02}}
     {targetBone {boneName=spine_01}}
     {targetBone {boneName=pelvis}}
     {targetBone {boneName=thigh_twist_01_l}}
     {targetBone {boneName=deform_glute_l}}
    ]
    constraintWeights [1,1,1,1,1,1,1,1,1,1]
   }
   n69 = "Constraint - Cannister (PPPPPPPP)"!(
    ComponentPose=@n68
    Alpha=1
    ConstraintWeights_0=0.4401
    ConstraintWeights_1:float=0.2801
    ConstraintWeights_2:float=0.1648
    ConstraintWeights_3:float=0.0848
    ConstraintWeights_4:float=0.0095
    ConstraintWeights_5:float=0.008
    ConstraintWeights_6:float=0.0084
    ConstraintWeights_7:float=0.0042
   ) -> (Pose) {
    alphaBoolBlend {blendOption=Linear}
    boneToModify {boneName=Cannister}
    constraintSetup [
     {targetBone {boneName=spine_05}}
     {targetBone {boneName=clavicle_l}}
     {targetBone {boneName=clavicle_r}}
     {targetBone {boneName=spine_04}}
     {targetBone {boneName=upperarm_twist_01_l}}
     {targetBone {boneName=spine_03}}
     {targetBone {boneName=spine_02}}
     {targetBone {boneName=deform_upperarm_bk_l}}
    ]
    constraintWeights [1,1,1,1,1,1,1,1]
   }
   n70 = "Constraint - dyn_skirt_bk_l_MSTR (RR)"!(ComponentPose=@n69, Alpha=1, ConstraintWeights_0=0.25, ConstraintWeights_1:float=0.75) -> (Pose) {
    alphaBoolBlend {blendOption=Linear}
    boneToModify {boneName=dyn_skirt_bk_l_MSTR}
    constraintSetup [{transformType=Rotation, targetBone {boneName=thigh_l}}, {transformType=Rotation, targetBone {boneName=pelvis}}]
    constraintWeights [1,1]
   }
   n71 = "Constraint - dyn_skirt_bk_r_MSTR (RR)"!(ComponentPose=@n70, Alpha=1, ConstraintWeights_0=0.25, ConstraintWeights_1:float=0.75) -> (Pose) {
    alphaBoolBlend {blendOption=Linear}
    boneToModify {boneName=dyn_skirt_bk_r_MSTR}
    constraintSetup [{transformType=Rotation, targetBone {boneName=thigh_r}}, {transformType=Rotation, targetBone {boneName=pelvis}}]
    constraintWeights [1,1]
   }
   n72 = Control Rig(Source=Component To Local(ComponentPose=@n71)) {
    alphaBoolBlend {blendOption=Linear}
    controlRigAssetReference {blueprintRigClass=&/F_MED_DistantEchoCastle_Style_Skirt_CtrlRig.F_MED_DistantEchoCastle_Style_Skirt_CtrlRig_C}
   }
   n73 = Get GravityOverride()
   n74 = RigidBody!(
    ComponentPose=Local To Component(LocalPose=@n72)
    Alpha:float=Get SkydiveRBValue() -> (SkydiveRBValue:double)
    OverrideWorldGravity=@n73
    ExternalForce
   ) -> (Pose) {
    bClampLinearTranslationLimitToRefPose=true
    bOverrideWorldGravity=true
    overridePhysicsAsset=&/F_MED_DistantEchoCastle_Style_Skirt_Physics
    simulationSpace=BaseBoneSpace
    alphaBoolBlend {blendOption=Linear}
    baseBoneRef {boneName=pelvis}
    componentAppliedLinearAccClamp=<5000 5000 5000>
    componentLinearAccScale {z=0.6}
    componentLinearVelScale {z=0.6}
    overrideWorldGravity {z=-980}
    simSpaceSettings {
     maxAngularAcceleration=500
     maxAngularVelocity=3
     maxLinearAcceleration=700
     maxLinearVelocity=500
     velocityScaleZ=0
     worldAlpha=0.25
     externalLinearDragV=<1 1>
    }
   }
   n75 = RigidBody(ComponentPose=@n74, Alpha=Get SkydiveRBStrapValue() -> (SkydiveRBStrapValue:double), OverrideWorldGravity=@n73) {
    bClampLinearTranslationLimitToRefPose=true
    bOverrideWorldGravity=true
    overridePhysicsAsset=&/F_MED_DistantEchoCastle_Style_Physics
    simulationSpace=BaseBoneSpace
    alphaBoolBlend {blendOption=Linear}
    baseBoneRef {boneName=root}
    componentAppliedLinearAccClamp=<5000 5000 5000>
    componentLinearAccScale {z=0.6}
    componentLinearVelScale {z=0.6}
    overrideWorldGravity {z=-980}
    simSpaceSettings {
     maxAngularAcceleration=500
     maxAngularVelocity=3
     maxLinearAcceleration=700
     maxLinearVelocity=500
     velocityScaleZ=0
     worldAlpha=0.25
     externalLinearDragV=<1 1>
    }
   }
   n76 = Get RightMovement() -> (RightMovement:double)
   n77 = "Transform (Modify) Bone - Bone: dyn_skirt_r_side_MSTR"(ComponentPose=@n75, Alpha=@n76, Translation="-3,0,0", Rotation="0,0,0", Scale="1,1,1") {
    rotationMode=BMM_Additive
    rotationSpace=BCS_BoneSpace
    translationMode=BMM_Additive
    alphaBoolBlend {blendOption=Linear}
    alphaScaleBiasClamp {bInterpResult=true}
    boneToModify {boneName=dyn_skirt_r_side_MSTR}
    translation {x=-3}
   }
   n78 = "Transform (Modify) Bone - Bone: dyn_skirt_bk_r_MSTR"(ComponentPose=@n77, Alpha=@n76, Translation="-3,0,0", Rotation="0,0,10", Scale="1,1,1") {
    rotationMode=BMM_Additive
    rotationSpace=BCS_BoneSpace
    translationMode=BMM_Additive
    alphaBoolBlend {blendOption=Linear}
    alphaScaleBiasClamp {bInterpResult=true}
    boneToModify {boneName=dyn_skirt_bk_r_MSTR}
    rotation {roll=10}
    translation {x=-3}
   }
   n79 = Get LeftMovement() -> (LeftMovement:double)
   n80 = "Transform (Modify) Bone - Bone: dyn_skirt_l_side_MSTR"(ComponentPose=@n78, Alpha=@n79, Translation="1,0,0", Rotation="0,0,0", Scale="1,1,1") {
    rotationMode=BMM_Additive
    rotationSpace=BCS_BoneSpace
    translationMode=BMM_Additive
    alphaBoolBlend {blendOption=Linear}
    alphaScaleBiasClamp {bInterpResult=true}
    boneToModify {boneName=dyn_skirt_l_side_MSTR}
    translation {x=1}
   }
   n81 = "Transform (Modify) Bone - Bone: dyn_skirt_bk_l_MSTR"(ComponentPose=@n80, Alpha=@n79, Translation="1,0,0", Rotation="0,0,0", Scale="1,1,1") {
    rotationMode=BMM_Additive
    rotationSpace=BCS_BoneSpace
    translationMode=BMM_Additive
    alphaBoolBlend {blendOption=Linear}
    alphaScaleBiasClamp {bInterpResult=true}
    boneToModify {boneName=dyn_skirt_bk_l_MSTR}
    translation {x=1}
   }
   n82 = Get CrouchNotMoving() -> (CrouchNotMoving:double)
   n83 = "Transform (Modify) Bone - Bone: dyn_skirt_l_side_MSTR"(ComponentPose=@n81, Alpha=@n82, Translation="0,0,0", Rotation="0,0,-10", Scale="0.65,1,1") {
    rotationMode=BMM_Additive
    rotationSpace=BCS_BoneSpace
    scaleMode=BMM_Additive
    translationMode=BMM_Additive
    alphaBoolBlend {blendOption=Linear}
    alphaScaleBiasClamp {bInterpResult=true}
    boneToModify {boneName=dyn_skirt_l_side_MSTR}
    rotation {roll=-10}
    scale {x=0.65}
   }
   n84 = "Transform (Modify) Bone - Bone: dyn_skirt_l_front_MSTR"!(ComponentPose=@n83, Alpha:float=@n82, Translation="0,0,-2", Rotation="-10,0,-30", Scale="0.65,1,1") -> (Pose) {
    rotationMode=BMM_Additive
    rotationSpace=BCS_BoneSpace
    scaleMode=BMM_Additive
    translationMode=BMM_Additive
    alphaBoolBlend {blendOption=Linear}
    alphaScaleBiasClamp {bInterpResult=true}
    boneToModify {boneName=dyn_skirt_l_front_MSTR}
    rotation {pitch=-10, roll=-30}
    scale {x=0.65}
    translation {z=-2}
   }
   n85 = "Transform (Modify) Bone - Bone: dyn_skirt_r_front2_MSTR"!(ComponentPose=@n84, Alpha:float=@n82, Translation="0,5,0", Rotation="0,0,0", Scale="1,1,1") -> (Pose) {
    rotationMode=BMM_Additive
    rotationSpace=BCS_BoneSpace
    scaleMode=BMM_Additive
    translationMode=BMM_Additive
    alphaBoolBlend {blendOption=Linear}
    alphaScaleBiasClamp {bInterpResult=true}
    boneToModify {boneName=dyn_skirt_r_front2_MSTR}
    translation {y=5}
   }
   n86 = "Transform (Modify) Bone - Bone: dyn_skirt_l_front2_MSTR"!(ComponentPose=@n85, Alpha:float=@n82, Translation="0,0,0", Rotation="-3,0,0", Scale="1,1,1") -> (Pose) {
    rotationMode=BMM_Additive
    rotationSpace=BCS_BoneSpace
    scaleMode=BMM_Additive
    translationMode=BMM_Additive
    alphaBoolBlend {blendOption=Linear}
    alphaScaleBiasClamp {bInterpResult=true}
    boneToModify {boneName=dyn_skirt_l_front2_MSTR}
    rotation {pitch=-3}
   }
   n87 = Get Is Sky Diving() -> (Is Sky Diving:double)
   n88 = "Transform (Modify) Bone - Bone: dyn_skirt_bk_MSTR"(ComponentPose=@n86, Alpha=@n87, Translation="0,0,0", Rotation="10,0,0", Scale="1,1,1") {
    rotationMode=BMM_Additive
    rotationSpace=BCS_BoneSpace
    scaleMode=BMM_Additive
    translationMode=BMM_Additive
    alphaBoolBlend {blendOption=Linear}
    alphaScaleBiasClamp {bInterpResult=true}
    boneToModify {boneName=dyn_skirt_bk_MSTR}
    rotation {pitch=10}
   }
   n89 = "Transform (Modify) Bone - Bone: dyn_skirt_strapLong_MSTR"!(ComponentPose=@n88, Alpha:float=@n87, Translation="0,0,0", Rotation="10,0,0", Scale="1,1,1") -> (Pose) {
    rotationMode=BMM_Additive
    rotationSpace=BCS_BoneSpace
    scaleMode=BMM_Additive
    translationMode=BMM_Additive
    alphaBoolBlend {blendOption=Linear}
    alphaScaleBiasClamp {bInterpResult=true}
    boneToModify {boneName=dyn_skirt_strapLong_MSTR}
    rotation {pitch=10}
   }
   n90 = "Transform (Modify) Bone - Bone: dyn_skirt_r_front2_MSTR"!(ComponentPose=@n89, bAlphaBoolEnabled:bool=Get bIsTargeting(), Translation="-2,3,3", Rotation="0,0,0", Scale="1,1,1") -> (Pose) {
    alphaInputType=Bool
    rotationMode=BMM_Additive
    rotationSpace=BCS_BoneSpace
    translationMode=BMM_Additive
    alphaBoolBlend {blendOption=Linear}
    alphaScaleBiasClamp {bInterpResult=true}
    boneToModify {boneName=dyn_skirt_r_front2_MSTR}
    translation=<-2 3 3>
   }
   n91 = Get bIsPlayingEmote()
   n92 = "Transform (Modify) Bone - Bone: dyn_skirt_r_front2_MSTR"(ComponentPose=@n90, bAlphaBoolEnabled=@n91, Translation="0,0,0", Rotation="10,0,0", Scale="1,1,1") {
    alphaInputType=Bool
    rotationMode=BMM_Additive
    rotationSpace=BCS_BoneSpace
    translationMode=BMM_Additive
    alphaBoolBlend {blendOption=Linear}
    alphaScaleBiasClamp {bInterpResult=true}
    boneToModify {boneName=dyn_skirt_r_front2_MSTR}
    rotation {pitch=10}
   }
   n93 = "Transform (Modify) Bone - Bone: dyn_skirt_l_front2_MSTR"!(ComponentPose=@n92, bAlphaBoolEnabled:bool=@n91, Translation="0,5,0", Rotation="10,0,0", Scale="1,1,1") -> (Pose) {
    alphaInputType=Bool
    rotationMode=BMM_Additive
    rotationSpace=BCS_BoneSpace
    translationMode=BMM_Additive
    alphaBoolBlend {blendOption=Linear}
    alphaScaleBiasClamp {bInterpResult=true}
    boneToModify {boneName=dyn_skirt_l_front2_MSTR}
    rotation {pitch=10}
    translation {y=5}
   }
   n94 = "Transform (Modify) Bone - Bone: attach_backpack"!(ComponentPose=@n93, bAlphaBoolEnabled:bool=Get isWearingBackbling(), Translation="0,2,0", Rotation="0,0,0", Scale="1,1,1") -> (Pose) {
    alphaInputType=Bool
    rotationSpace=BCS_BoneSpace
    translationMode=BMM_Additive
    translationSpace=BCS_BoneSpace
    alphaBoolBlend {blendOption=Linear}
    alphaScaleBiasClamp {bInterpResult=true}
    boneToModify {boneName=attach_backpack}
    translation {y=2}
   }
   n95 = "Transform (Modify) Bone - Bone: CapeD_MSTR_l"(ComponentPose=@n94, Alpha=Get using Pistol() -> (using Pistol:double), Translation="0,0,8", Rotation="0,0,0", Scale="1,1,1") {
    rotationSpace=BCS_BoneSpace
    translationMode=BMM_Additive
    translationSpace=BCS_BoneSpace
    alphaBoolBlend {blendOption=Linear}
    alphaScaleBiasClamp {bInterpResult=true}
    boneToModify {boneName=CapeD_MSTR_l}
    translation {z=8}
   }
   Save cached pose 'Start_Cache'(Pose=Component To Local(ComponentPose=@n95)) {cacheName=Start_Cache}
   Get bIsParachuteOpened()
   Get bIsSkydiveFloating()
   Get bIsPlayingMeleeAnim()
   Get isWearingCapeBool()
  ]}
 EventGraph {nodes [
   n1 = Cloth Controller Function DEPRECATED!(
    execute
    CharacterPartAnimInstance:Custom Character Part Anim Instance Object Reference=Self-Reference()
    ClothingAssetName:String=Base_Cape_DistantEchoCastle
    Cloth Paramaters:Cloth Params Struct Structure=Get Cloth Paramaters() -> (Cloth Paramaters:Cloth Params Struct Structure)
    Cloth Core Settings:Cloth Core Setting Struct Structure=Get Cloth Core Settings() -> (Cloth Core Settings:Cloth Core Setting Struct Structure)
    ClothUpAxis:Vector="0,0,1"
    OnGround_Moving_SpeedLimit:double=2
    OnGround_Moving_AngularVel_Max:double=0.05
    OnGround_Moving_FictitiousVel_Max:double=0.6
    LinVelScale_FwdBwd_Clamp:double=0
    LinVelScale_UpDwn_Clamp:double=0.7
    DeltaTime:double=0
    GameTimeInSeconds:double=GetGameTimeInSeconds()
    WindGustsParams:Wind Gust Params Struct Structure=Get Wind Gust Params() -> (Wind Gust Params:Wind Gust Params Struct Structure)
    WindGustLocalSettings:Wind Gust Local Setting Struct Structure=Get Gust Local Settings() -> (Gust Local Settings:Wind Gust Local Setting Struct Structure)
   ) -> (
    then
    CurrentStateOut:String
    Cloth Settings Out:Cloth Setting Struct Structure
    Cloth Core Settings Out:Cloth Core Setting Struct Structure
    WindGustSettings Out:Wind Gust Setting Struct Structure
    WindGustLocalSettings Out:Wind Gust Local Setting Struct Structure
    GustStrength - WindVelocityOut:Vector
    LinearVelVector-DELETE LATER:Vector
   )
   > Set Gust Local Settings(Gust Local Settings:Wind Gust Local Setting Struct Structure=@n1.WindGustLocalSettings Out) -> (Output_Get:Wind Gust Local Setting Struct Structure)
   Event BlueprintUpdateAnimation()
   > n2 = Sequence!(execute) -> (then_0, then_1, then_2:Exec, then_3:Exec, then_4:Exec, then_5:Exec, then_6:Exec, then_7:Exec, then_8:Exec, then_9:Exec)
   > n3 = Player Movement Function DEPRECATED!(
    execute
    CharacterPartAnimInstance:Custom Character Part Anim Instance Object Reference=Self-Reference()
    X Dot Product:double=1
    Y Dot Product:double=1
    Z Dot Product:double=1
    DrawDebug:bool=false
    Debug Scalar:double=50
   ) -> (then, X Movement:double, Y Movement:double, Z Movement:double)
   n4 = "To Float (Boolean)"(InBool=NOT Boolean(A=Get bIsSprinting()))
   n5 = "To Float (Boolean)"(InBool=NOT Boolean(A=Get bIsInVehicle() -> (bIsInVehicle:bool)))
   n6 = "To Float (Boolean)"(InBool=NOT Boolean(A=Get bIsBackpedaling() -> (bIsBackpedaling:bool)))
   n7 = "To Float (Boolean)"(InBool=NOT Boolean(A=Get bIsSwimSprinting() -> (bIsSwimSprinting:bool)))
   n8 = "To Float (Boolean)"(InBool=NOT Boolean(A=Get bIsSurfaceSwimming()))
   n9 = float + float(A=float * float(A="Absolute (Float)"(A=@n3.Y Movement), B=0.2), B=MapRangeClamped(Value=@n3, InRangeA=0, InRangeB=1)) {is=title K2Node_CommutativeAssociativeBinaryOperator}
   n10 = "To Float (Boolean)"(InBool=NOT Boolean(A=Get bIsSkyDiving()))
   > Set RightMovement(RightMovement:double=float * float!(
     A=MapRangeClamped(Value=@n9, InRangeA=0, InRangeB=1.2, OutRangeB=0.5)
     B=2.5
     C:double=@n10
     D:double=@n4
     E:double=@n6
     F:double=@n7
     G:double=@n8
     H:double=@n5
    ) -> (ReturnValue) {is=title K2Node_PromotableOperator}) -> (Output_Get:double)
   n11 = float + float(
    A=float * float(A="Absolute (Float)"(A=@n3.Y Movement), B=0.2) {is=title K2Node_CommutativeAssociativeBinaryOperator}
    B=MapRangeClamped(Value=@n3, InRangeA=0, InRangeB=-1, OutRangeB=0.3)
   )
   > Set LeftMovement(LeftMovement:double=float * float(A=MapRangeClamped(Value=@n11, InRangeA=0, InRangeB=1.2, OutRangeB=0.5), B=8.5, C=@n10, D=@n4, E=@n6, F=@n7, G=@n8, H=@n5) {is=title K2Node_PromotableOperator}) -> (Output_Get:double)
   > Set BwdMovement(BwdMovement:double=MapRangeClamped(Value=@n3.Y Movement, InRangeA=0, InRangeB=-0.7)) -> (Output_Get:double)
   n12 = AND Boolean!(A=Get frontendNoEmote(), B=Get CapeOn(), C:bool=NOT Boolean(A=Get bIsPlayingEmote())) -> (ReturnValue)
   > Set CapeOnInFrontend(CapeOnInFrontend:bool=@n12) -> (Output_Get:bool)
   > Set BackblingCapeFrontend(BackblingCapeFrontend:bool=AND Boolean!(A=@n12, B=Get isWearingBackbling()) -> (ReturnValue)) -> (Output_Get:bool)
   > Set isPickaxing(isPickaxing:float="To Float (Boolean)"(InBool=Get bIsPlayingMeleeAnim())) -> (Output_Get:float)
   n13 = Get bRigidBodyEnabled()
   > Set RigidBodiesOffInFrontend(RigidBodiesOffInFrontend:bool=AND Boolean(A=Get bIsInFrontEnd(), B=NOT Boolean(A=@n13))) -> (Output_Get:bool)
   n14 = NOT Boolean(A=Get bIsMoving2D())
   n15 = Get bIsCrouching()
   > Set CrouchNotMoving(CrouchNotMoving:double="To Float (Boolean)"(InBool=AND Boolean(A=@n15, B=@n14))) -> (Output_Get:double)
   n16 = Get bIsSkyDiving()
   > Set Is Sky Diving(Is Sky Diving:double="To Float (Boolean)"(InBool=@n16)) -> (Output_Get:double)
   > n17 = Branch(Condition=@n16)
   > Set SkydiveRBValue(SkydiveRBValue:double=0.25) -> (Output_Get:double)
   Set SkydiveRBValue(execute=@n17.else, SkydiveRBValue=1)
   n18 = Gravity Override State Function DEPRECATED(
    execute=@n2.then_1
    CharacterPartAnimInstance=Self-Reference()
    GameTimeInSeconds=GetGameTimeInSeconds()
    GravityOverrideParams=Get Gravity Override Params()
   )
   > Set GravityOverride(GravityOverride=@n18)
   n19 = Branch(execute=@n2.then_2, Condition=OR Boolean(A=Get bIsSkyDiving(), B=Get isWearingBackbling()))
   > Set SkydiveRBStrapValue(SkydiveRBStrapValue:double=0.1) -> (Output_Get:double)
   Set SkydiveRBStrapValue(execute=@n19.else, SkydiveRBStrapValue=0.75)
   n20 = Cast To FortPlayerPawn(execute=@n2.then_3, Object=GetOwningActor())
   > n21 = GetVariantChannelActiveIndexForItemActor!(
    execute
    Item:Athena Cosmetic Item Definition Object Reference=/BRCosmetics/Athena/Items/Cosmetics/Characters/Character_DistantEchoCastle
    Actor:Actor Object Reference=@n20
    Channel:Gameplay Tag Structure=(TagName="Cosmetics.Variant.Channel.Parts")
    VariantList:Array of Gameplay Tag Structures=Make Array("[0]"=(TagName="Cosmetics.Variant.Property.Stage3"))
    bErrorIfChannelMissing:bool=false
    bErrorIfTagNotInArray:bool=false
   ) -> (then, Index:int=0, bIndexIsValid:bool, bFoundChannel:bool)
   > n22 = Branch(Condition=@n21.bIndexIsValid)
   > Set CapeOn(CapeOn:bool=true) -> (Output_Get:bool)
   Set CapeOn(execute=@n22.else, CapeOn=false)
   Set SkydiveCapeRB(execute=@n2.then_4, SkydiveCapeRB:double=float * float!(A=0.5, B=Get Is Sky Diving()) -> (ReturnValue)) -> (Output_Get:double)
   n23 = Get bIsPlayingEmote()
   Set frontendNoEmote(execute=@n2.then_5, frontendNoEmote:bool=AND Boolean(A=Get bIsInFrontEnd(), B=NOT Boolean(A=@n23))) -> (Output_Get:bool)
   Set ar_firingPosition(execute=@n2.then_6, ar_firingPosition:double=0) -> (Output_Get:double)
   > n24 = Switch on EFortWeaponCoreAnimation!(execute, Selection:EFortWeaponCoreAnimation Enum=Get WeaponCoreAnim!() -> (WeaponCoreAnim)) -> (
    Melee:Exec
    Pistol:Exec
    Shotgun:Exec
    PaperBlueprint:Exec
    Rifle:Exec
    MeleeOneHand:Exec
    MachinePistol:Exec
    RocketLauncher:Exec
    GrenadeLauncher:Exec
    GoingCommando:Exec
    AssaultRifle:Exec
    TacticalShotgun:Exec
    SniperRifle:Exec
    TrapPlacement:Exec
    ShoulderLauncher:Exec
    AbilityDecoTool:Exec
    Crossbow:Exec
    C4:Exec
    RemoteControl:Exec
    DualWield:Exec
    AR_BullPup:Exec
    AR_ForwardGrip:Exec
    MedPackPaddles:Exec
    SMG_P90:Exec
    AR_DrumGun:Exec
    Consumable_Small:Exec
    Consumable_Large:Exec
    Balloon:Exec
    MountedTurret:Exec
    CreativeTool:Exec
    ExplosiveBow:Exec
    AshtonIndigo:Exec
    AshtonChicago:Exec
    MeleeDualWield:Exec
    Unarmed:Exec
   )
   Set using Pistol(
    execute=[
     @n24.AR_BullPup
     @n24.AR_DrumGun
     @n24.AR_ForwardGrip
     @n24.AbilityDecoTool
     @n24.AshtonChicago
     @n24.AshtonIndigo
     @n24.AssaultRifle
     @n24.Balloon
     @n24.C4
     @n24.Consumable_Large
     @n24.Consumable_Small
     @n24.CreativeTool
     @n24.Crossbow
     @n24.DualWield
     @n24.ExplosiveBow
     @n24.GoingCommando
     @n24.GrenadeLauncher
     @n24.MachinePistol
     @n24.MedPackPaddles
     @n24.Melee
     @n24.MeleeDualWield
     @n24.MeleeOneHand
     @n24.MountedTurret
     @n24.PaperBlueprint
     @n24.RemoteControl
     @n24.Rifle
     @n24.RocketLauncher
     @n24.SMG_P90
     @n24.Shotgun
     @n24.ShoulderLauncher
     @n24.SniperRifle
     @n24.TacticalShotgun
     @n24.TrapPlacement
    ]
    using Pistol:double=0
   ) -> (Output_Get:double)
   Set using Pistol(execute=@n24.Pistol, using Pistol=1)
   n25 = CheckForBackbling()
   n26 = Set isWearingBackbling(execute=[@n31.then, @n2.then_7], isWearingBackbling=OR Boolean(A=@n25, B=@n25.bIsBackblingLong))
   > n27 = Set isWearingCapeBool(isWearingCapeBool:bool=CheckForBackbling().bIsBackblingACape) -> (Output_Get:bool)
   > n28 = Branch(Condition=OR Boolean(A=@n26, B=@n27))
   > n29 = Get Dynamic Material Instance(SkeletalMeshComponent=GetPartSkeletalMeshComponent())
   > SetScalarParameterValue(self=@n29, ParameterName=Hide Element 02, Value=1)
   SetScalarParameterValue(execute=@n28.else, self=@n29, ParameterName=Hide Element 02, Value=0)
   Event BlueprintInitializeAnimation()
   n30 = GetPartSkeletalMeshComponent()
   > Set bResetAfterTeleport!(execute, bResetAfterTeleport:bool=false, self:Skeletal Mesh Component Object Reference=@n30) -> (then, Output_Get:bool)
   n31 = Event OnCharacterCustomizationLoaded()
   Event OnEndFalling()
   > Branch(Condition=Get resetClothLandingFromZilpine() -> (resetClothLandingFromZilpine:bool))
   n32 = Get ChaosClothingInteractor() -> (ChaosClothingInteractor:Chaos Clothing Interactor Object Reference)
   > ResetAndTeleport!(execute, self:Chaos Clothing Interactor Object Reference=@n32, bReset:bool=true, bTeleport:bool=true) -> (then)
   > Set resetClothLandingFromZilpine(resetClothLandingFromZilpine:bool=false) -> (Output_Get:bool)
   Event OnBeginSkydiving()
   > Set SkydivingAlpha(SkydivingAlpha:Rotator="0,-65,0") -> (Output_Get:Rotator)
   n33 = Event BlueprintProcessAnimInputEvent()
   > n34 = Branch(Condition=IsNamed(self=@n33.AnimInputEvent, InEventName=aie.Player.Zipline.Entered))
   > Delay(Duration=0.1)
   > ResetAndTeleport(self=@n32, bReset=true, bTeleport=true)
   Branch(execute=@n34.else, Condition=IsNamed(self=@n33.AnimInputEvent, InEventName=aie.Player.Zipline.Exited))
   > Set resetClothLandingFromZilpine(resetClothLandingFromZilpine=true)
   Event OnEndSkydiving()
   > Set SkydivingAlpha(SkydivingAlpha="0,-10,0")
   n35 = Branch(Condition=Get bIsSkyDiving())
   > n36 = Branch(Condition="float > float"(A=Get PawnVelocityZ(), B=50) {is=title K2Node_CallFunction})
   > Set BwdMovement(BwdMovement=float * float(A=Get BwdMovement(), B=0) {is=title K2Node_CommutativeAssociativeBinaryOperator})
   Set BwdMovement(execute=[@n35.else, @n36.else], BwdMovement=float * float(A=Get BwdMovement(), B=1))
   n37 = Branch(Condition=Get bIsInFrontEnd())
   > Set BwdMovement(BwdMovement=0.5)
   n38 = Player Movement Function DEPRECATED(
    execute=@n37.else
    CharacterPartAnimInstance=Self-Reference()
    X Dot Product=1
    Y Dot Product=1
    Z Dot Product=1
    DrawDebug=false
    Debug Scalar=50
   )
   > n39 = Branch(Condition=Get bIsDBNOCarried())
   > Set BwdMovement(BwdMovement=0.5)
   n40 = Branch(execute=@n39.else, Condition=AND Boolean(A=Get bIsTargeting(), B=Get bIsCrouching()))
   n41 = MapRangeClamped(
    Value=float + float(A="Absolute (Float)"(A=@n38), B=MapRangeClamped(Value=@n38.Y Movement, InRangeA=-1, InRangeB=1, OutRangeA=0.3))
    InRangeA=0
    InRangeB=1.65
   )
   > Set BwdMovement(BwdMovement=float * float(A=@n41, B=0.25) {is=title K2Node_PromotableOperator})
   Set BwdMovement(execute=@n40.else, BwdMovement=@n41)
   n42 = GetClothingSimulationInteractor!(self:Skeletal Mesh Component Object Reference=GetOwningComponent(), ClothingAssetName:Name=None) -> (ReturnValue:Clothing Simulation Interactor Object Reference)
   Is Valid(InputObject=@n42)
   n43 = MakeLiteralString(Value=Base_Cape_01)
   > n44 = Cast To ChaosClothingInteractor!(
    execute
    Object=GetClothingInteractor!(self:Clothing Simulation Interactor Object Reference=@n42, ClothingAssetName:Name=String To Name(InString=@n43)) -> (ReturnValue:Clothing Interactor Object Reference)
   ) -> (then, CastFailed, AsChaos Clothing Interactor:Chaos Clothing Interactor Object Reference) {is=title K2Node_DynamicCast}
   > Set ChaosClothingInteractor(ChaosClothingInteractor:Chaos Clothing Interactor Object Reference=@n44) -> (Output_Get:Chaos Clothing Interactor Object Reference)
   Set TeleportDistanceThreshold!(execute, TeleportDistanceThreshold:float=300, self:Skeletal Mesh Component Object Reference=@n30) -> (then, Output_Get:float)
   > Set TeleportRotationThreshold!(execute, TeleportRotationThreshold:float=25, self:Skeletal Mesh Component Object Reference=@n30) -> (then, Output_Get:float)
   n45 = GetOwnerSkeletalMeshComponent()
   DrawVectorFromJointFunction!(
    execute
    SkelMeshComponent:Scene Component Object Reference=@n45
    JointName:Name=pelvis
    LineEnd:Vector=Break Cloth Setting Struct!(ClothSettingStruct:"Cloth Setting Struct Structure (by ref)"=@n1.Cloth Settings Out) -> (
     LinearVelocityScale_7_#1:Vector=~
     AngularVelocityScale_6_#2:float=~
     FictitiousAngularScale_10_#3:float=~
     GravityOverride_15_#4:Vector=~
     DragCoefficient_23_#5:Vector 2D Structure
     LiftCoefficient_24_#6:Vector 2D Structure
     JointName_27_#7:Name
    ) {is=title K2Node_BreakStruct}.GravityOverride_15_#4
    JointOffset:Vector="0, 0, 0"
    LineColor:Linear Color Structure=(R=1,G=0,B=0,A=1)
    Thickness:double=2
    Foreground:EDrawDebugSceneDepthPriorityGroup Enum=Foreground
   ) -> (then)
   DrawJointDebugFunction!(
    execute
    SkelMeshComponent:Scene Component Object Reference=@n30
    JointOffset:Vector="0, 0, 0"
    JointName:Array of Names=Make Array("[0]":Name=root) -> (Array:Array of Names)
    Scale:double=10
    Duration:double=0
    Thickness:double=0.7
    DebugDraw:bool=true
    Foreground:EDrawDebugSceneDepthPriorityGroup Enum=Foreground
   ) -> (then)
   Debug Cloth Teleport Threshold Function!(
    execute
    MinDistance:double=80
    MinRotation:double=25
    SkelMeshComponent:Scene Component Object Reference=@n30
    PrintToScreen:bool=true
    BoneName:Name=root
    InVector:"Vector (by ref)"=Get TeleportDistanceThreshold_Local() -> (TeleportDistanceThreshold_Local:Vector)
    PawnDeltaRotation:Rotator=Get PawnDeltaRotation() -> (PawnDeltaRotation:Rotator)
    PrintToScreenDuration:double=6
   ) -> (then, Distance:double, Rotation:double)
   DrawVectorFromJointFunction(
    SkelMeshComponent=@n45
    JointName=spine_05
    LineEnd="0, 0, 0"
    JointOffset="0, 0, 0"
    LineColor=(R=1,G=0,B=0,A=1)
    Thickness=2
    Foreground=Foreground
   )
   Branch(Condition=Get bClothEnabled())
   Get bIsSkyDiving()
   Set RbanOrCloth(RbanOrCloth:bool=OR Boolean(A=Get bRigidBodyEnabled(), B=Get bClothEnabled())) -> (Output_Get:bool)
  ]}
]
variables [
 ar_firingPosition {type=number}
 BackblingCapeFrontend {type=boolean}
 BwdMovement {type=number}
 CapeOn {type=boolean}
 CapeOnInFrontend {type=boolean}
 ChaosClothingInteractor {type=ChaosClothingInteractor, default=null}
 Cloth Core Settings {
  type=ClothCoreSettingStruct
  default {
   collisionThickness=1.5
   dampingCoefficient=0.1
   frictionCoefficient=0.2
   localDampingCoefficient=0
   numberIterations=2
   numberSubsteps=1
   selfCollisionThickness=0
   animDriveDamping=<0 1>
   animDriveStiffness=<0 1>
   areaStiffness=<1 1>
   bendingStiffness=<1 1>
   edgeStiffness=<1 1>
   tetherStiffness=<1 1>
  }
 }
 Cloth Paramaters {
  type=ClothParamsStruct
  default {
   dBNO {
    angularVelocityScale=0.6
    fictitiousAngularScale=0.1
    jointName=pelvis
    dragCoefficient=<0.2 1>
    gravityOverride=<-980 400 0>
    liftCoefficient=<0 0.5>
    linearVelocityScale=<0.3 0.3 0.1>
   }
   dBNO_Carried {
    angularVelocityScale=0.6
    fictitiousAngularScale=0.1
    jointName=pelvis
    dragCoefficient=<0.2 1>
    gravityOverride=<-980 -200 0>
    liftCoefficient=<0 0.5>
    linearVelocityScale=<0.3 0.3 0.1>
   }
   dBNO_Impulse {
    angularVelocityScale=0
    fictitiousAngularScale=0
    jointName=pelvis
    dragCoefficient=<0 0.2>
    gravityOverride=<-3000 -2000 0>
    liftCoefficient=<0 0.1>
    linearVelocityScale=<0.3 0.1 0.3>
   }
   dBNO_RollBwd {
    angularVelocityScale=0
    fictitiousAngularScale=0
    jointName=pelvis
    dragCoefficient=<0 0.2>
    gravityOverride=<-2000 0 0>
    liftCoefficient=<0 0.1>
    linearVelocityScale=<0.3 0.1 0.3>
   }
   dBNO_RollFwd {
    angularVelocityScale=0
    fictitiousAngularScale=0
    jointName=pelvis
    dragCoefficient=<0 0.2>
    gravityOverride=<-980 0 0>
    liftCoefficient=<0 0.1>
    linearVelocityScale=<0.3 0.1 0.3>
   }
   dBNO_Tumbling {
    angularVelocityScale=0
    fictitiousAngularScale=0
    jointName=pelvis
    dragCoefficient=<0 0.2>
    gravityOverride=<-2500 -300 0>
    liftCoefficient=<0 0.1>
    linearVelocityScale=<0.3 0.1 0.3>
   }
   emoteOrMelee {
    angularVelocityScale=0.6
    fictitiousAngularScale=0.1
    jointName=root
    dragCoefficient=<0 0.3>
    gravityOverride=<0 0 -980>
    liftCoefficient=<0 0.2>
    linearVelocityScale=<0.6 0.6 0.5>
   }
   front_End {
    angularVelocityScale=0.05
    fictitiousAngularScale=0.6
    jointName=root
    dragCoefficient=<0.2 0.2>
    gravityOverride=<0 -200 -980>
    liftCoefficient=<0.1 0.1>
    linearVelocityScale=<0.1 0.1 0.1>
   }
   onGround_Crouch_Moving {
    angularVelocityScale=0.1
    fictitiousAngularScale=0.3
    jointName=root
    dragCoefficient=<2 2>
    gravityOverride=<0 0 -980>
    liftCoefficient=<0.5 0.5>
    linearVelocityScale=<0.6 0.6 0.5>
   }
   onGround_Falling {
    angularVelocityScale=0.6
    fictitiousAngularScale=0
    jointName=root
    dragCoefficient=<0 0.2>
    gravityOverride=<0 -500 -980>
    liftCoefficient=<0 0.1>
    linearVelocityScale=<2 1 0.3>
   }
   onGround_Moving {
    angularVelocityScale=0.6
    fictitiousAngularScale=0.1
    jointName=root
    dragCoefficient=<0.2 3>
    gravityOverride=<0 0 -980>
    liftCoefficient=<0 0.3>
    linearVelocityScale=<0.3 1 1>
   }
   onGround_Standing {
    angularVelocityScale=0.05
    fictitiousAngularScale=1
    jointName=root
    dragCoefficient=<0.2 2.5>
    gravityOverride=<0 0 -980>
    liftCoefficient=<0 0.3>
    linearVelocityScale=<0.3 1 1>
   }
   rail_Grinding {
    angularVelocityScale=0.2
    fictitiousAngularScale=0
    jointName=root
    dragCoefficient=<0.1 0.2>
    gravityOverride=<50 -500 -400>
    liftCoefficient=<0 0.5>
    linearVelocityScale=<0.1 0.1 0.1>
   }
   riding_Moving {
    angularVelocityScale=0.05
    fictitiousAngularScale=1
    jointName=root
    dragCoefficient=<0.5 0.5>
    gravityOverride=<0 0 -980>
    liftCoefficient=<0.5 0.5>
    linearVelocityScale=<0.4 0.6 0.3>
   }
   riding_Moving_Targeting {
    angularVelocityScale=0.05
    fictitiousAngularScale=0.7
    jointName=root
    dragCoefficient=<0.5 0.5>
    gravityOverride=<0 0 -980>
    liftCoefficient=<0.5 0.5>
    linearVelocityScale=<0.2 0.6 0.2>
   }
   riding_Standing {
    angularVelocityScale=0.05
    fictitiousAngularScale=1
    jointName=root
    dragCoefficient=<0.5 0.5>
    gravityOverride=<0 -450 -980>
    liftCoefficient=<0.5 0.5>
    linearVelocityScale=<0.6 0.6 0.3>
   }
   skydive_DivingDown {
    angularVelocityScale=0.1
    fictitiousAngularScale=0
    jointName=pelvis
    dragCoefficient=<0 1>
    gravityOverride=<-450 0 0>
    liftCoefficient=<0 0.1>
    linearVelocityScale=<0.1 0.1 0.1>
   }
   skydive_FloatingDown {
    angularVelocityScale=0.1
    fictitiousAngularScale=0
    jointName=root
    dragCoefficient=<0 2>
    gravityOverride=<0 0 980>
    liftCoefficient=<0 0.4>
    linearVelocityScale=<0.2 0.2 0.1>
   }
   skydive_GoingUp {
    angularVelocityScale=0.1
    fictitiousAngularScale=0
    jointName=pelvis
    dragCoefficient=<0.1 0.1>
    gravityOverride=<-10000 -400 0>
    liftCoefficient=<0.05 0.05>
    linearVelocityScale=<0.01 0.1 0.1>
   }
   skydive_Parachuting {
    angularVelocityScale=0.6
    fictitiousAngularScale=0.1
    jointName=root
    dragCoefficient=<0 2.5>
    gravityOverride=<0 -600 -500>
    liftCoefficient=<0 0.5>
    linearVelocityScale=<0.2 0.2 0.3>
   }
   skydive_Surfing {
    angularVelocityScale=0.6
    fictitiousAngularScale=0.1
    jointName=root
    dragCoefficient=<1.5 1.5>
    gravityOverride=<0 -600 -500>
    liftCoefficient=<1 1>
    linearVelocityScale=<0.1 0.1 0.3>
   }
   swimming {
    angularVelocityScale=0.6
    fictitiousAngularScale=0
    jointName=root
    dragCoefficient=<2.5 2.5>
    gravityOverride=<0 0 0>
    liftCoefficient=<0.1 0.1>
    linearVelocityScale=<0.6 0.6 0.5>
   }
   tactical_Sprint {
    angularVelocityScale=0.6
    fictitiousAngularScale=0.1
    jointName=root
    dragCoefficient=<2 2>
    gravityOverride=<0 0 -980>
    liftCoefficient=<0.5 0.5>
    linearVelocityScale=<0.6 0.6 0.5>
   }
   ziplining {
    angularVelocityScale=0.6
    fictitiousAngularScale=0
    jointName=root
    dragCoefficient=<0 1.5>
    gravityOverride=<0 0 -980>
    liftCoefficient=<0 0.3>
    linearVelocityScale=<0.4 0.4 0.5>
   }
  }
 }
 CrouchNotMoving {type=number}
 frontendNoEmote {type=boolean}
 Gravity Override Params {type=GravityOverrideParamsStruct, default {skydive_FloatingDown {gravityOverride=<0 -700 100>}}}
 GravityOverride {type=Vector}
 Gust Local Settings {type=WindGustLocalSettingStruct, default {elapsedTime=0, finalGustStrengthLocal=0, gustDuration=0, gustStrength=0, timeBetweenGusts=0}}
 Is Sky Diving {type=number}
 isPickaxing {type=number}
 isWearingBackbling {type=boolean}
 isWearingCapeBool {type=boolean}
 LeftMovement {type=number}
 RbanOrCloth {type=boolean}
 resetClothLandingFromZilpine {type=boolean}
 RightMovement {type=number}
 RigidBodiesOffInFrontend {type=boolean}
 SkydiveCapeRB {type=number}
 SkydiveRBStrapValue {type=number}
 SkydiveRBValue {type=number}
 SkydivingAlpha {type=Rotator}
 TeleportDistanceThreshold_Local {type=Vector}
 using Pistol {type=number}
 Wind Gust Params {
  type=WindGustParamsStruct
  default {
   dBNO {
    maxDownInterpSpeed=5
    maxGustDuration=2
    maxGustStrength=7
    maxTimeBetweenGusts=7
    maxUpInterpSpeed=5
    minDownInterpSpeed=2
    minGustDuration=0.5
    minGustStrength=3
    minTimeBetweenGusts=0
    minUpInterpSpeed=2
    windTurbulenceScalarA=0.1
    windTurbulenceScalarB=1
    windAmplitude=<0 0 0>
    windFrequency=<0 0 0>
    windOffset=<0 0 0>
   }
   dBNO_Carried {
    maxDownInterpSpeed=5
    maxGustDuration=2
    maxGustStrength=7
    maxTimeBetweenGusts=7
    maxUpInterpSpeed=5
    minDownInterpSpeed=2
    minGustDuration=0.5
    minGustStrength=3
    minTimeBetweenGusts=0
    minUpInterpSpeed=2
    windTurbulenceScalarA=0.1
    windTurbulenceScalarB=1
    windAmplitude=<0 0 0>
    windFrequency=<0 0 0>
    windOffset=<0 0 0>
   }
   dBNO_Impulse {
    maxDownInterpSpeed=5
    maxGustDuration=2
    maxGustStrength=7
    maxTimeBetweenGusts=7
    maxUpInterpSpeed=5
    minDownInterpSpeed=2
    minGustDuration=0.5
    minGustStrength=3
    minTimeBetweenGusts=0
    minUpInterpSpeed=2
    windTurbulenceScalarA=0.1
    windTurbulenceScalarB=1
    windAmplitude=<0 0 0>
    windFrequency=<0 0 0>
    windOffset=<0 0 0>
   }
   dBNO_RollBwd {
    maxDownInterpSpeed=5
    maxGustDuration=2
    maxGustStrength=7
    maxTimeBetweenGusts=7
    maxUpInterpSpeed=5
    minDownInterpSpeed=2
    minGustDuration=0.5
    minGustStrength=3
    minTimeBetweenGusts=0
    minUpInterpSpeed=2
    windTurbulenceScalarA=0.1
    windTurbulenceScalarB=1
    windAmplitude=<0 0 0>
    windFrequency=<0 0 0>
    windOffset=<0 0 0>
   }
   dBNO_RollFwd {
    maxDownInterpSpeed=5
    maxGustDuration=2
    maxGustStrength=7
    maxTimeBetweenGusts=7
    maxUpInterpSpeed=5
    minDownInterpSpeed=2
    minGustDuration=0.5
    minGustStrength=3
    minTimeBetweenGusts=0
    minUpInterpSpeed=2
    windTurbulenceScalarA=0.1
    windTurbulenceScalarB=1
    windAmplitude=<0 0 0>
    windFrequency=<0 0 0>
    windOffset=<0 0 0>
   }
   dBNO_Tumbling {
    maxDownInterpSpeed=5
    maxGustDuration=2
    maxGustStrength=7
    maxTimeBetweenGusts=7
    maxUpInterpSpeed=5
    minDownInterpSpeed=2
    minGustDuration=0.5
    minGustStrength=3
    minTimeBetweenGusts=0
    minUpInterpSpeed=2
    windTurbulenceScalarA=0.1
    windTurbulenceScalarB=1
    windAmplitude=<0 0 0>
    windFrequency=<0 0 0>
    windOffset=<0 0 0>
   }
   emoteOrMelee {
    maxDownInterpSpeed=5
    maxGustDuration=2
    maxGustStrength=7
    maxTimeBetweenGusts=7
    maxUpInterpSpeed=5
    minDownInterpSpeed=2
    minGustDuration=0.5
    minGustStrength=3
    minTimeBetweenGusts=0
    minUpInterpSpeed=2
    windTurbulenceScalarA=0.1
    windTurbulenceScalarB=1
    windAmplitude=<30 25 50>
    windFrequency=<2 3 5>
    windOffset=<-300 -150 0>
   }
   front_End {
    maxDownInterpSpeed=5
    maxGustDuration=2
    maxGustStrength=10
    maxTimeBetweenGusts=7
    maxUpInterpSpeed=5
    minDownInterpSpeed=2
    minGustDuration=0.5
    minGustStrength=5
    minTimeBetweenGusts=0
    minUpInterpSpeed=2
    windTurbulenceScalarA=0.1
    windTurbulenceScalarB=1
    windAmplitude=<10 30 50>
    windFrequency=<15 6 7>
    windOffset=<0 -200 -250>
   }
   onGround_Crouch_Moving {
    maxDownInterpSpeed=5
    maxGustDuration=2
    maxGustStrength=7
    maxTimeBetweenGusts=7
    maxUpInterpSpeed=5
    minDownInterpSpeed=2
    minGustDuration=0.5
    minGustStrength=3
    minTimeBetweenGusts=0
    minUpInterpSpeed=2
    windTurbulenceScalarA=0.1
    windTurbulenceScalarB=1
    windAmplitude=<100 25 75>
    windFrequency=<1.5 0.5 7.5>
    windOffset=<0 -100 -175>
   }
   onGround_Falling {
    maxDownInterpSpeed=5
    maxGustDuration=8
    maxGustStrength=7
    maxTimeBetweenGusts=8
    maxUpInterpSpeed=5
    minDownInterpSpeed=2
    minGustDuration=0.2
    minGustStrength=3
    minTimeBetweenGusts=1
    minUpInterpSpeed=2
    windTurbulenceScalarA=0.1
    windTurbulenceScalarB=0.8
    windAmplitude=<20 65 42>
    windFrequency=<5 6 8.6>
    windOffset=<30 -100 -300>
   }
   onGround_Moving {
    maxDownInterpSpeed=5
    maxGustDuration=8
    maxGustStrength=7
    maxTimeBetweenGusts=8
    maxUpInterpSpeed=5
    minDownInterpSpeed=2
    minGustDuration=0.2
    minGustStrength=3
    minTimeBetweenGusts=1
    minUpInterpSpeed=2
    windTurbulenceScalarA=0.1
    windTurbulenceScalarB=0.8
    windAmplitude=<20 65 42>
    windFrequency=<5 6 8.6>
    windOffset=<30 -100 -300>
   }
   onGround_Standing {
    maxDownInterpSpeed=5
    maxGustDuration=8
    maxGustStrength=7
    maxTimeBetweenGusts=8
    maxUpInterpSpeed=5
    minDownInterpSpeed=2
    minGustDuration=0.2
    minGustStrength=2
    minTimeBetweenGusts=1
    minUpInterpSpeed=2
    windTurbulenceScalarA=0.1
    windTurbulenceScalarB=0.8
    windAmplitude=<0 15 28>
    windFrequency=<0 2.3 3.7>
    windOffset=<0 -30 -150>
   }
   rail_Grinding {
    maxDownInterpSpeed=5
    maxGustDuration=8
    maxGustStrength=7
    maxTimeBetweenGusts=8
    maxUpInterpSpeed=5
    minDownInterpSpeed=2
    minGustDuration=0.2
    minGustStrength=3
    minTimeBetweenGusts=1
    minUpInterpSpeed=2
    windTurbulenceScalarA=0.1
    windTurbulenceScalarB=0.8
    windAmplitude=<20 65 42>
    windFrequency=<5 6 8.6>
    windOffset=<50 -500 -400>
   }
   riding_Moving {
    maxDownInterpSpeed=5
    maxGustDuration=2
    maxGustStrength=7
    maxTimeBetweenGusts=7
    maxUpInterpSpeed=5
    minDownInterpSpeed=2
    minGustDuration=0.5
    minGustStrength=3
    minTimeBetweenGusts=0
    minUpInterpSpeed=2
    windTurbulenceScalarA=0.1
    windTurbulenceScalarB=1
    windAmplitude=<0 0 0>
    windFrequency=<0 0 0>
    windOffset=<0 0 0>
   }
   riding_Moving_Targeting {
    maxDownInterpSpeed=5
    maxGustDuration=2
    maxGustStrength=7
    maxTimeBetweenGusts=7
    maxUpInterpSpeed=5
    minDownInterpSpeed=2
    minGustDuration=0.5
    minGustStrength=3
    minTimeBetweenGusts=0
    minUpInterpSpeed=2
    windTurbulenceScalarA=0.1
    windTurbulenceScalarB=1
    windAmplitude=<0 0 0>
    windFrequency=<0 0 0>
    windOffset=<0 0 0>
   }
   riding_Standing {
    maxDownInterpSpeed=5
    maxGustDuration=2
    maxGustStrength=7
    maxTimeBetweenGusts=7
    maxUpInterpSpeed=5
    minDownInterpSpeed=2
    minGustDuration=0.5
    minGustStrength=3
    minTimeBetweenGusts=0
    minUpInterpSpeed=2
    windTurbulenceScalarA=0.1
    windTurbulenceScalarB=1
    windAmplitude=<0 0 0>
    windFrequency=<0 0 0>
    windOffset=<0 0 0>
   }
   skydive_DivingDown {
    maxDownInterpSpeed=5
    maxGustDuration=8
    maxGustStrength=7
    maxTimeBetweenGusts=8
    maxUpInterpSpeed=5
    minDownInterpSpeed=2
    minGustDuration=0.2
    minGustStrength=3
    minTimeBetweenGusts=1
    minUpInterpSpeed=2
    windTurbulenceScalarA=0.1
    windTurbulenceScalarB=4
    windAmplitude=<0 25 35>
    windFrequency=<0 4 10>
    windOffset=<-450 -100 100>
   }
   skydive_FloatingDown {
    maxDownInterpSpeed=5
    maxGustDuration=8
    maxGustStrength=1
    maxTimeBetweenGusts=8
    maxUpInterpSpeed=5
    minDownInterpSpeed=2
    minGustDuration=0.2
    minGustStrength=0
    minTimeBetweenGusts=1
    minUpInterpSpeed=2
    windTurbulenceScalarA=0.5
    windTurbulenceScalarB=1
    windAmplitude=<0 0 0>
    windFrequency=<0 0 0>
    windOffset=<0 0 0>
   }
   skydive_GoingUp {
    maxDownInterpSpeed=5
    maxGustDuration=8
    maxGustStrength=7
    maxTimeBetweenGusts=8
    maxUpInterpSpeed=5
    minDownInterpSpeed=2
    minGustDuration=0.2
    minGustStrength=3
    minTimeBetweenGusts=1
    minUpInterpSpeed=2
    windTurbulenceScalarA=0.1
    windTurbulenceScalarB=0.8
    windAmplitude=<0 0 0>
    windFrequency=<0 0 0>
    windOffset=<0 -500 300>
   }
   skydive_Parachuting {
    maxDownInterpSpeed=5
    maxGustDuration=8
    maxGustStrength=1
    maxTimeBetweenGusts=8
    maxUpInterpSpeed=5
    minDownInterpSpeed=2
    minGustDuration=5
    minGustStrength=0
    minTimeBetweenGusts=5
    minUpInterpSpeed=2
    windTurbulenceScalarA=0.2
    windTurbulenceScalarB=1
    windAmplitude=<25 50 75>
    windFrequency=<8 6 2>
    windOffset=<0 -200 -150>
   }
   skydive_Surfing {
    maxDownInterpSpeed=5
    maxGustDuration=8
    maxGustStrength=1
    maxTimeBetweenGusts=16
    maxUpInterpSpeed=5
    minDownInterpSpeed=2
    minGustDuration=5
    minGustStrength=0
    minTimeBetweenGusts=5
    minUpInterpSpeed=2
    windTurbulenceScalarA=0.2
    windTurbulenceScalarB=1
    windAmplitude=<25 50 75>
    windFrequency=<8 6 2>
    windOffset=<250 -200 -200>
   }
   swimming {
    maxDownInterpSpeed=5
    maxGustDuration=2
    maxGustStrength=7
    maxTimeBetweenGusts=7
    maxUpInterpSpeed=5
    minDownInterpSpeed=2
    minGustDuration=0.5
    minGustStrength=3
    minTimeBetweenGusts=0
    minUpInterpSpeed=2
    windTurbulenceScalarA=0.1
    windTurbulenceScalarB=1
    windAmplitude=<10 25 20>
    windFrequency=<1.5 0.5 7.5>
    windOffset=<0 -150 50>
   }
   tactical_Sprint {
    maxDownInterpSpeed=5
    maxGustDuration=2
    maxGustStrength=7
    maxTimeBetweenGusts=7
    maxUpInterpSpeed=5
    minDownInterpSpeed=2
    minGustDuration=0.5
    minGustStrength=3
    minTimeBetweenGusts=0
    minUpInterpSpeed=2
    windTurbulenceScalarA=0.1
    windTurbulenceScalarB=1
    windAmplitude=<0 0 0>
    windFrequency=<0 0 0>
    windOffset=<0 0 0>
   }
   ziplining {
    maxDownInterpSpeed=5
    maxGustDuration=8
    maxGustStrength=7
    maxTimeBetweenGusts=8
    maxUpInterpSpeed=5
    minDownInterpSpeed=2
    minGustDuration=0.2
    minGustStrength=3
    minTimeBetweenGusts=1
    minUpInterpSpeed=2
    windTurbulenceScalarA=0.1
    windTurbulenceScalarB=0.8
    windAmplitude=<20 65 42>
    windFrequency=<5 6 8.6>
    windOffset=<30 -100 -300>
   }
  }
 }
]
